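# Material function that adds flickering or sine-pulsed emissive to incoming Material Attributes, halving BaseColor and blending by SFXMask.
# Static switches pick layered sine/cosine jitter or a DebugTimeSine pulse with optional rounding to on/off.
# Optional InteriorGlow multiplies flicker by a banded camera-to-normal falloff mask.
# The result drives an intensity lerp scaled by overall brightness, but both lerp ends read Flicker Min, leaving Max unused.
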MaterialFunction /WildEstateTech/Environment/BaseMaterials/MaterialFunctions/MF_WildEstate_Flicker
graph {nodes [
  n1 = GetMaterialAttributes(None=NamedRerouteDeclaration(None=FunctionInput() {inputType=FunctionInput_MaterialAttributes}) {name=InMat, nodeColor=<rgba 0.8047 1 0 1>}) {attributeGetTypes [#1, #2]}
  n2 = Multiply(A=@n1.BaseColor) {constB=0.5}
  n3 = MaterialFunctionCall(Float3=FunctionInput(Preview=Constant3Vector() {constant=<rgba 0 1 1 0>}) {desc="R=Flicker Min, G=Flicker Max, B=Flicker overallBrightness", inputName=EmissiveIntensity, sortPriority=3}) {materialFunction=&BreakOutFloat3Components}
  n4 = NamedRerouteUsage()
  n5 = NamedRerouteUsage()
  n6 = Multiply(A=Time(), B=@n5)
  n7 = Add(A=Cosine(None=Multiply(A=@n6) {constB=4.5}), B=@n6)
  n8 = Multiply(A=Sine(None=Multiply(A=@n7) {constB=3})) {constB=0.66}
  n9 = Multiply(A=Sine(None=Multiply(A=@n6) {constB=4})) {constB=0.5}
  n10 = Cosine(None=Multiply(A=@n6) {constB=3.6})
  n11 = Sine(None=@n6)
  n12 = ConstantBiasScale(None=Multiply(A=Add(A=@n8, B=@n9), B=Add(A=@n10, B=@n11)))
  n13 = Multiply() {constB=1.8889}
  n14 = Sine(None=Multiply() {constB=0.75})
  n15 = OneMinus(None=@n4)
  n16 = Multiply(A=Add(A=Cosine(None=@n13), B=Multiply(A=@n14) {constB=0.66}), B=@n15)
  n17 = Multiply(A=Power(Base=@n4) {constExponent=0.25}, B=Subtract(A=Saturate(None=@n12), B=Add(A=@n16, B=@n15)))
  n18 = MaterialFunctionCall(Speed=@n5) {materialFunction=&DebugTimeSine}
  n19 = Round(None=@n18.Result)
  n20 = FunctionInput() {
   desc="True=Clamp values on and off for harsher transition, False=linear animation"
   inputName=Flicker_SinewaveClamp
   inputType=FunctionInput_StaticBool
   sortPriority=4
  }
  n21 = FunctionInput() {desc="True=flicker jitter controls, False=sinewave controls", inputName=Flicker_Modulate, inputType=FunctionInput_StaticBool, sortPriority=5}
  n22 = Saturate(None=StaticSwitch(True=Saturate(None=@n17), False=StaticSwitch(True=@n19, False=@n18.Result, Value=@n20), Value=@n21))
  n23 = DotProduct(A=CameraVectorWS(), B=PixelNormalWS())
  n24 = NamedRerouteUsage()
  n25 = Saturate(None=Multiply(A=Power(Base=@n23, Exp=@n24), B=NamedRerouteUsage()))
  n26 = Constant() {r=4}
  n27 = Divide(A=Floor(None=Multiply(A=@n25, B=@n26) {constB=5}), B=Subtract(A=@n26))
  n28 = FunctionInput() {
   desc="True=Interior glow falloff controls, False=no interior masking controls"
   inputName=InteriorGlow
   inputType=FunctionInput_StaticBool
   sortPriority=6
  }
  n29 = StaticSwitch(True=Multiply(A=@n22, B=Saturate(None=@n27)), False=@n22, Value=@n28)
  n30 = Multiply(A=@n1.EmissiveColor, B=Multiply(A=LinearInterpolate(A=@n3.R, B=@n3.R, Alpha=@n29), B=@n3.B))
  FunctionOutput(None=BlendMaterialAttributes(
    A=NamedRerouteUsage()
    B=SetMaterialAttributes!(MaterialAttributes=@n1.MaterialAttributes, BaseColor=@n2, EmissiveColor=@n30) {attributeSetTypes [#1, #2]}
    Alpha=FunctionInput() {inputName=SFXMask, inputType=FunctionInput_Scalar, sortPriority=10}
   ))
  MaterialFunctionCall(Emissive=@n30) {materialFunction=&MF_FX_EmissiveAdjustments}
  n31 = FunctionInput(Preview=Constant4Vector() {constant=<rgba 1 1 0.7 1>}.RGBA) {
   desc="R=Flicker Speed, G=Flicker Jitter Amount, B=Interior Mask Contrast, A=Interior mask Core Brightness"
   inputName=FlickerAmounts
   inputType=FunctionInput_Vector4
   sortPriority=2
  }
  n32 = MaterialFunctionCall(Float4=@n31) {materialFunction=&BreakOutFloat4Components}
  NamedRerouteDeclaration(None=@n32.R) {name=FlickerSpeed, nodeColor=<rgba 0 0.4375 1 1>}
  NamedRerouteDeclaration(None=@n32.G) {name=FlickerJitterAmount, nodeColor=<rgba 0 0.4375 1 1>}
  NamedRerouteDeclaration(None=@n32.B) {name=InteriorContrast, nodeColor=<rgba 0 0.4375 1 1>}
  NamedRerouteDeclaration(None=@n32.A) {name=InteriorCore, nodeColor=<rgba 0 0.4375 1 1>}
 ]}
