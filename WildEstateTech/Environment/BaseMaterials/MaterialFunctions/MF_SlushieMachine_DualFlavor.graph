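# Material function that adds an animated two-color slush fill and machine shake to incoming material attributes, behind a static switch.
# Perlin noise and a BumpOffset parallax layer build wavy foreground and background fill masks with edge lines, tinting BaseColor and Emissive.
# World Position Offset layers sine sway, view-transformed micro jitter and bounds-masked vertical jitter, using actor-position-offset time, also output separately.
# Custom primitive data drives fill level and shake intensity.
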MaterialFunction /WildEstateTech/Environment/BaseMaterials/MaterialFunctions/MF_SlushieMachine_DualFlavor
graph {nodes [
  n1 = GetMaterialAttributes(None=NamedRerouteDeclaration(None=FunctionInput() {bUsePreviewValueAsDefault=true, inputName=MaterialIn, inputType=FunctionInput_MaterialAttributes}) {name=MatIn, nodeColor=<rgba 1 0.1875 0 1>}) {attributeGetTypes [#1, #2]}
  n2 = NamedRerouteUsage()
  n3 = Max(A=NamedRerouteUsage(), B=NamedRerouteUsage())
  n4 = LinearInterpolate(A=@n1.BaseColor, B=@n2, Alpha=@n3)
  n5 = LinearInterpolate(A=@n1.EmissiveColor, B=@n2, Alpha=@n3)
  n6 = Multiply(
   A=NamedRerouteUsage()
   B=ScalarParameter() {defaultValue=10, group=Slush Machine, parameterName=SlushMachine.WPO.JitterSpeedVert, sortPriority=40}
  ) {constB=10}
  n7 = Multiply(
   A=Multiply(
    A=Sine(None=@n6)
    B=ScalarParameter() {defaultValue=0.01, group=Slush Machine, parameterName=SlushMachine.WPO.JitterDistanceVert, sortPriority=40}
   ) {constB=0.01}
   B=Constant3Vector() {constant=<rgba 0 0 1 0>}
  )
  n8 = SmoothStep(
   Min=ScalarParameter() {group=Slush Machine, parameterName=SlushMachine.WPO.MaskMin}
   Max=ScalarParameter() {defaultValue=1, group=Slush Machine, parameterName=SlushMachine.WPO.MaskMax}
   Value=Saturate(None=MaterialFunctionCall() {materialFunction=&BoundingBoxBased_0-1_UVW}.B)
  )
  n9 = ScalarParameter() {
   bUseCustomPrimitiveData=true
   desc=master control for the Shaking for the Slush Mashine WPO
   group=Slush Machine
   parameterName=BP_SlushMachineWPOIntensity
   primitiveDataIndex=2
   sortPriority=40
  }
  n10 = Multiply(A=Add(A=NamedRerouteUsage(), B=LinearInterpolate(B=@n7, Alpha=@n8)), B=Saturate(None=@n9))
  FunctionOutput(None=StaticSwitchParameter(
    True=SetMaterialAttributes!(MaterialAttributes=@n1.MaterialAttributes, BaseColor=@n4, EmissiveColor=@n5, WorldPositionOffset=@n10) {attributeSetTypes [#1, #2, #3]}
    False=NamedRerouteUsage()
   ) {group=Wild Estate Tech Art FX, parameterName=UseDualSlushMachineFX, sortPriority=1}) {outputName=Material}
  FunctionOutput(None=StaticSwitchParameter(True=@n10, False=Constant()) {group=Wild Estate Tech Art FX, parameterName=UseDualSlushMachineFX, sortPriority=1}) {outputName=Shake WPO, sortPriority=4}
  n11 = NamedRerouteUsage()
  n12 = ConstantBiasScale(None=Sine(None=@n11))
  n13 = Sine(None=Multiply(A=Time()) {constB=5})
  n14 = ConstantBiasScale(None=Sine(None=Multiply(A=@n11) {constB=0.5}))
  Multiply(
   A=Multiply(A=SmoothStep(Value=@n12) {constMax=0.7, constMin=0.3, desc=Shake Mask}, B=ConstantBiasScale(None=@n13))
   B=LinearInterpolate(
    A=Constant3Vector() {constant=<rgba 1 0 0 0>}
    B=Constant3Vector() {constant=<rgba 0 1 0 0>}
    Alpha=SmoothStep(Value=@n14) {constMax=0.55, constMin=0.45, desc=Shake Mask}
   )
  )
  n15 = Add(A=NamedRerouteUsage()) {constB=0.62}
  n16 = Sine(None=@n15)
  n17 = Add(A=@n15, B=Multiply(A=ConstantBiasScale(None=@n16)) {constB=0.71})
  n18 = SmoothStep(Value=ConstantBiasScale(None=Sine(None=@n17))) {constMax=0.7, constMin=0.3}
  n19 = Add(A=@n15) {constB=0.34}
  n20 = SmoothStep(Value=ConstantBiasScale(None=Sine(None=@n19))) {constMax=0.7, constMin=0.3}
  n21 = NamedRerouteUsage()
  n22 = Sine(None=@n21)
  n23 = Add(A=@n21, B=Multiply(A=ConstantBiasScale(None=@n22)) {constB=0.3})
  n24 = SmoothStep(Value=ConstantBiasScale(None=Sine(None=@n23))) {constMax=0.7, constMin=0.3}
  n25 = Add(A=@n21) {constB=0.34}
  n26 = SmoothStep(Value=ConstantBiasScale(None=Sine(None=@n25))) {constMax=0.7, constMin=0.3}
  n27 = AppendVector(A=Add(A=AppendVector(A=@n18, B=@n20), B=AppendVector(A=@n24, B=@n26)), B=Constant())
  n28 = ScalarParameter() {defaultValue=10, group=Slush Machine, parameterName=SlushMachine.WPO.PosDistance, sortPriority=40}
  n29 = NamedRerouteUsage()
  n30 = ScalarParameter() {defaultValue=10, group=Slush Machine, parameterName=SlushMachine.WPO.PosMicroSpeed, sortPriority=40}
  n31 = Multiply(
   A=Sine(None=Multiply(A=@n29, B=@n30) {constB=10})
   B=ScalarParameter() {defaultValue=0.5, group=Slush Machine, parameterName=SlushMachine.WPO.PosMicroDistance, sortPriority=40}
  ) {constB=0.5}
  n32 = Multiply(A=@n31, B=Transform(None=Constant3Vector() {constant=<rgba 0 1 0 0>}) {transformSourceType=TRANSFORMSOURCE_View})
  n33 = Multiply(A=@n31, B=Transform(None=Constant3Vector() {constant=<rgba 1 0 0 0>}) {transformSourceType=TRANSFORMSOURCE_View})
  NamedRerouteDeclaration(None=Add(A=Multiply(A=@n27, B=@n28), B=Add(A=@n32, B=@n33))) {name=newWPO, nodeColor=<rgba 1 0.7031 0 1>}
  n34 = ActorPositionWS()
  NamedRerouteDeclaration(None=Add(A=Time(), B=Length(None=@n34) {is=class MaterialExpressionLength})) {name=RandTime, nodeColor=<rgba 1 0.7031 0 1>}
  n35 = ScalarParameter() {defaultValue=-0.369, group=Slush Machine, parameterName=SlushMachine.Slush.FillMin}
  n36 = ScalarParameter() {defaultValue=-0.15, group=Slush Machine, parameterName=SlushMachine.Slush.FillMax}
  n37 = ScalarParameter() {
   bUseCustomPrimitiveData=true
   defaultValue=1
   desc=master control for the Shaking for the Slush Mashine WPO
   group=Slush Machine
   parameterName=BP_SlushMachineFillAlpha
   primitiveDataIndex=1
   sortPriority=40
  }
  n38 = Constant() {r=0.5}
  n39 = MaterialFunctionCall(Input=@n37, Input High=@n38) {materialFunction=&RemapValueRange}
  n40 = MaterialFunctionCall(Input=@n37, Input Low=@n38) {materialFunction=&RemapValueRange}
  n41 = Saturate(None=LinearInterpolate(A=Clamp(None=@n39.Result), B=Clamp(None=@n40.Result), Alpha=NamedRerouteUsage()))
  n42 = Add(A=TextureCoordinate(), B=AppendVector(A=Constant(), B=LinearInterpolate(A=@n35, B=@n36, Alpha=@n41)))
  n43 = TextureCoordinate()
  n44 = Multiply(A=NamedRerouteUsage(), B=ScalarParameter() {defaultValue=0.05, group=Slush Machine, parameterName=SlushMachine.Slush.NoiseSpeed})
  n45 = TextureSample(UVs=Multiply(A=Add(A=@n43, B=@n44), B=ScalarParameter() {defaultValue=3, group=Slush Machine, parameterName=SlushMachine.Slush.NoiseTiling})) {texture=/Game/Global/Textures/Default/Noise/3xPerlin_M}
  n46 = Add(
   A=@n42
   B=Multiply(A=ScalarParameter() {defaultValue=0.05, group=Slush Machine, parameterName=SlushMachine.Slush.NoiseIntensityBackground}, B=@n45.R) {constB=0.1}
  )
  n47 = MaterialFunctionCall(UV Channel=@n46) {materialFunction=&LinearGradient}
  NamedRerouteDeclaration(None=Saturate(None=Round(None=@n47.VGradient))) {name=ForegroundMask, nodeColor=<rgba 1 0.0415 0.8158 1>}
  n48 = ScalarParameter() {defaultValue=0.02, group=Slush Machine, parameterName=SlushMachine.Slush.NoiseIntensityForeground}
  n49 = BumpOffset(
   Coordinate=Add(A=@n42, B=Multiply(A=@n45.G, B=@n48) {constB=0.05})
   Height=LinearInterpolate(B=ScalarParameter() {defaultValue=-1.5, group=Slush Machine, parameterName=SlushMachine.Slush.BumpOffset.Height}, Alpha=@n41)
  )
  n50 = ScalarParameter() {defaultValue=0.05, group=Slush Machine, parameterName=SlushMachine.Slush.BackgroundLayerOffset}
  n51 = Add(
   A=MaterialFunctionCall(UV Channel=@n49) {materialFunction=&LinearGradient}.VGradient
   B=@n50
  ) {constB=0.05}
  NamedRerouteDeclaration(None=Saturate(None=Round(None=@n51))) {name=BackgroundMask, nodeColor=<rgba 1 0 0.5859 1>}
  n52 = NamedRerouteUsage()
  n53 = VectorParameter() {group=Slush Machine, parameterName=SlushMachine.Slush.Color.B, defaultValue=<rgba 0.8073 0.6319 0.0158 1>}
  n54 = VectorParameter() {group=Slush Machine, parameterName=SlushMachine.Slush.Color.A, defaultValue=<rgba 0.7396 0.1861 0.0205 1>}
  n55 = NamedRerouteUsage()
  n56 = MaterialFunctionCall() {materialFunction=&LinearGradient}
  n57 = Saturate(None=Power(
    Base=Add(A=@n56.VGradient) {constB=0.05}
    Exp=ScalarParameter() {defaultValue=16, group=Slush Machine, parameterName=SlushMachine.Slush.2ToneContrast, sortPriority=40}
   ))
  n58 = Multiply(
   A=LinearInterpolate(A=Multiply(A=@n52) {constB=0.3}, Alpha=NamedRerouteUsage())
   B=LinearInterpolate(
    A=LinearInterpolate(A=@n53.RGB, B=@n54.RGB, Alpha=@n55)
    B=LinearInterpolate(A=@n53.RGB, B=@n54.RGB, Alpha=@n57)
    Alpha=ScalarParameter() {defaultValue=1, group=Slush Machine, parameterName=SlushMachine.Slush.2Tone, sortPriority=40}
   )
  )
  n59 = VectorParameter() {group=Slush Machine, parameterName=SlushMachine.Slush.LineColor, defaultValue=<rgba 0 0 0 1>}
  n60 = AppendVector(
   A=Constant()
   B=Multiply(A=ScalarParameter() {defaultValue=0.02, group=Slush Machine, parameterName=SlushMachine.Slush.LineThickness}) {constB=0.1}
  )
  n61 = MaterialFunctionCall(UV Channel=Add(A=@n49, B=@n60)) {materialFunction=&LinearGradient}
  n62 = Saturate(None=Round(None=Add(A=@n50, B=@n61.VGradient) {constB=0.05}))
  n63 = NamedRerouteUsage()
  n64 = Saturate(None=Subtract(A=Subtract(A=@n62, B=@n63), B=NamedRerouteUsage()))
  n65 = Add(A=@n46, B=@n60)
  n66 = Saturate(None=Round(None=MaterialFunctionCall(UV Channel=@n65) {materialFunction=&LinearGradient}.VGradient))
  n67 = NamedRerouteUsage()
  NamedRerouteDeclaration(None=LinearInterpolate(A=LinearInterpolate(A=@n58, B=@n59.RGB, Alpha=@n64) {constB=0}, B=@n59.RGB, Alpha=Subtract(A=@n66, B=@n67)) {constB=0}) {name=Color, nodeColor=<rgba 0 0.1328 1 1>}
  n68 = VectorParameter() {
   group=Slush Machine
   parameterName=SlushMachine.Slush.AreaMaskParams.A
   channelNames {a=Density, b=Radius, g=RadialMaskPosY, r=RadialMaskPosX}
   defaultValue=<rgba 0.5 0.5 0.1 100>
  }
  NamedRerouteDeclaration(None=MaterialFunctionCall(CenterPosition=AppendVector(A=@n68.RadialMaskPosX, B=@n68.RadialMaskPosX), Radius=@n68.Radius, Density=@n68.Radius) {materialFunction=&RadialGradientExponential}.RadialGradientExponential) {name=MaskA, nodeColor=<rgba 0 1 0.0625 1>}
  n69 = VectorParameter() {
   group=Slush Machine
   parameterName=SlushMachine.Slush.AreaMaskParams.B
   channelNames {a=Density, b=Radius, g=RadialMaskPosY, r=RadialMaskPosX}
   defaultValue=<rgba 0.5 0.5 0.1 100>
  }
  NamedRerouteDeclaration(None=MaterialFunctionCall(CenterPosition=AppendVector(A=@n69.RadialMaskPosX, B=@n69.RadialMaskPosX), Radius=@n69.Radius, Density=@n69.Radius) {materialFunction=&RadialGradientExponential}.RadialGradientExponential) {name=MaskB, nodeColor=<rgba 1 0 0.2578 1>}
  ScalarParameter() {defaultValue=0.05, group=Slush Machine, parameterName=SlushMachine.Slush.2ToneOffset, sortPriority=40}
  n70 = ScalarParameter() {
   bUseCustomPrimitiveData=true
   desc=master control for the Shaking for the Slush Mashine WPO
   group=Slush Machine
   parameterName=BP_SlushMachineFillAlpha
   primitiveDataIndex=1
   sortPriority=40
  }
  StaticSwitchParameter(True=@n70, False=ScalarParameter() {defaultValue=0.75, group=Slush Machine, parameterName=SlushMachine.Slush.FillAlpha, sliderMax=1}) {group=Slush Machine, parameterName=Use Custom Prim ForIntensity And Fill Level, sortPriority=1}
 ]}
