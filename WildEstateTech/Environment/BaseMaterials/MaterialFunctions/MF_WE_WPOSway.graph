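# Material function that animates meshes with World Position Offset sway, bob and rotation, adding the offset to incoming material attributes.
# Sine-driven, location-offset sway rotates vertices around an adjustable object pivot, with further variants for constant Z spin and PerInstanceRandom-derived axis and speed.
# Each variant corrects normals through FixRotateAboutAxisNormals.
# Static switches toggle the effect and select between random rotation, simple sway and a third variant.
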MaterialFunction /WildEstateTech/Environment/BaseMaterials/MaterialFunctions/MF_WE_WPOSway
graph {
 properties {description="This function is used to give assets rotation on the X, Y, and Z-axis. Normalized rotation along with fixed normals. "}
 nodes [
  n1 = NamedRerouteDeclaration(None=FunctionInput() {inputType=FunctionInput_MaterialAttributes}) {name=MatIn, nodeColor=<rgba 0.4062 1 0 1>}
  n2 = GetMaterialAttributes(None=@n1) {attributeGetTypes [#1]}
  n3 = SetMaterialAttributes!(
   MaterialAttributes=@n2.MaterialAttributes
   WorldPositionOffset=Add(A=@n2.WorldPositionOffset, B=NamedRerouteUsage())
   Normal=NamedRerouteUsage()
  ) {attributeSetTypes [#1, #2]}
  n4 = GetMaterialAttributes(None=@n1) {attributeGetTypes [#1]}
  n5 = NamedRerouteUsage()
  n6 = GetMaterialAttributes(None=@n1) {attributeGetTypes [#1, #2]}
  n7 = NamedRerouteUsage()
  n8 = StaticSwitchParameter(
   True=SetMaterialAttributes(MaterialAttributes=@n4.MaterialAttributes, WorldPositionOffset=Add(A=@n4.WorldPositionOffset, B=@n5), Normal=NamedRerouteUsage()) {attributeSetTypes [#1, #2]}
   False=SetMaterialAttributes(MaterialAttributes=@n6.MaterialAttributes, WorldPositionOffset=Add(A=@n6.WorldPositionOffset, B=@n7), Normal=NamedRerouteUsage()) {attributeSetTypes [#1, #2]}
  ) {group=WPO Sway, parameterName=SimpleSwayRotation, sortPriority=-99}
  FunctionOutput(None=StaticSwitchParameter(True=StaticSwitchParameter(True=@n3, False=@n8) {group=WPO Sway, parameterName=RandomRotation, sortPriority=-99}, False=NamedRerouteUsage()) {group=WPO, parameterName=Use WPO Sway, sortPriority=-100})
  n9 = Multiply(A=PerInstanceRandom()) {constB=100}
  n10 = ConstantBiasScale(None=Truncate!(None=@n9) {is=class MaterialExpressionTruncate}) {bias=-50, scale=0.02}
  n11 = Multiply(A=Frac(None=@n9)) {constB=100}
  n12 = ConstantBiasScale(None=Truncate(None=@n11)) {bias=-50, scale=0.02}
  n13 = Frac(None=@n11)
  n14 = ConstantBiasScale(None=Truncate(None=Multiply(A=@n13) {constB=100})) {bias=-50, scale=0.02}
  NamedRerouteDeclaration(None=Normalize(VectorInput=MaterialFunctionCall(R=@n10, G=@n12, B=@n14) {materialFunction=&AppendMany}.RGB)) {name=RotationAxis, nodeColor=<rgba 1 0.4453 0 1>}
  NamedRerouteDeclaration(None=@n6.Normal) {name=InNormal, nodeColor=<rgba 0.125 0 1 1>}
  n15 = VectorParameter() {group=WPO Sway, parameterName=WPO_Movement_Max, sortPriority=300, defaultValue=<rgba 180 180 180 1>}
  n16 = MaterialFunctionCall(Color=Divide(A=ObjectPositionWS(), B=ScalarParameter() {defaultValue=64, group=WPO Sway, parameterName=PositionOffset, sortPriority=300})) {materialFunction=&SplitComponents}
  n17 = VectorParameter() {group=WPO Sway, parameterName=WPO_Speed, sortPriority=300, defaultValue=<rgba 0.04 0.0175 0.03 2>}
  n18 = Add(A=Frac(None=Multiply(A=@n16.R, B=@n16.R)), B=Multiply(A=Time(), B=Multiply(A=@n17.RGB, B=@n17.RGB)))
  NamedRerouteDeclaration(None=LinearInterpolate(
    A=VectorParameter() {group=WPO Sway, parameterName=WPO_Movement_Min, sortPriority=300, defaultValue=<rgba -1 1 0 1>}.RGB
    B=@n15.RGB
    Alpha=Sine(None=@n18)
   )) {name=Bob, nodeColor=<rgba 0 0.7656 1 1>}
  n19 = ScalarParameter() {defaultValue=0.1, group=WPO Sway, parameterName=RotationSpeed, sortPriority=60}
  n20 = ConstantBiasScale(None=PerInstanceRandom())
  NamedRerouteDeclaration(None=Multiply(A=Time(), B=Multiply(A=@n19, B=@n20)) {constA=5, constB=0.1}) {name=RotationAngle, nodeColor=<rgba 1 0.6797 0 1>}
  n21 = ScalarParameter() {defaultValue=0.1, group=WPO Sway, parameterName=RotationSpeedX}
  n22 = ScalarParameter() {defaultValue=0.1, group=WPO Sway, parameterName=RotationSpeedZ}
  n23 = MaterialFunctionCall(Float3=VectorParameter() {group=WPO Sway, parameterName=RotationAxis, defaultValue=<rgba 1 1 0 1>}.RGB) {materialFunction=&BreakOutFloat3Components}
  n24 = Multiply(
   A=Add(A=MaterialFunctionCall(Speed=@n21) {materialFunction=&DebugTimeSine}.Result, B=@n22)
   B=@n23.R
  )
  n25 = ScalarParameter() {defaultValue=0.15, group=WPO Sway, parameterName=RotationSpeedY}
  n26 = Multiply(
   A=Add(A=MaterialFunctionCall(Speed=@n25) {materialFunction=&DebugTimeSine}.Result, B=@n22)
   B=@n23.G
  )
  n27 = Normalize(VectorInput=MaterialFunctionCall!(
    VectorToRotate=ObjectOrientation()
    LookAtVector=MaterialFunctionCall(X=@n24, Y=@n26, Z=@n23.B) {materialFunction=&MakeFloat3}.Result
    RestingVector
   ) {materialFunction=/Engine/Functions/Engine_MaterialFunctions01/Vectors/RotateVector}.Result)
  n28 = Time()
  n29 = MaterialFunctionCall(Float3=ObjectPositionWS()) {materialFunction=&BreakOutFloat3Components}
  n30 = Add(A=@n29.R, B=@n29.R)
  n31 = Add(A=@n28, B=Divide(A=Add(A=@n30, B=@n29.B), B=ScalarParameter() {defaultValue=0.1, group=WPO Sway, parameterName=TimeOffsetAmount}))
  n32 = Sine(None=Multiply(
    A=StaticSwitchParameter(True=@n31, False=@n28) {defaultValue=true, group=WPO Sway, parameterName=TimeLocationOffset}
    B=ScalarParameter() {defaultValue=0.5, group=WPO Sway, parameterName=SwaySpeed}
   ))
  n33 = Sine(None=@n31) {period=6}
  n34 = Multiply(
   A=StaticSwitchParameter(True=ConstantBiasScale(None=@n32), False=@n32) {group=WPO Sway, parameterName=BiasSineWave}
   B=LinearInterpolate(
    A=ScalarParameter() {defaultValue=0.005, group=WPO Sway, parameterName=SwayAmountMin}
    B=ScalarParameter() {defaultValue=0.02, group=WPO Sway, parameterName=SwayAmountMax}
    Alpha=ConstantBiasScale(None=@n33)
   )
  )
  n35 = VectorParameter() {group=WPO Sway, parameterName=PivotOffset, defaultValue=<rgba 0 0 0 1>}
  n36 = RotateAboutAxis(
   NormalizedRotationAxis=@n27
   RotationAngle=@n34
   PivotPoint=Add(
    A=MaterialFunctionCall() {materialFunction=&ObjectPivotPoint}.Object Pivot Location
    B=Transform(None=@n35.RGB) {transformSourceType=TRANSFORMSOURCE_Local}
   )
   Position=WorldPosition().XYZ
  )
  NamedRerouteDeclaration(None=Add(A=NamedRerouteUsage(), B=@n36)) {name=WPOFFinal, nodeColor=<rgba 0 1 0.0625 1>}
  NamedRerouteDeclaration(None=MaterialFunctionCall(
    VectorToRotate=NamedRerouteUsage()
    LookAtVector=MaterialFunctionCall(Rotation Axis=@n27, Rotation Angle=@n34) {materialFunction=&FixRotateAboutAxisNormals}.New Tangent Space Vertex Normal
   ) {materialFunction=&RotateVector}.Result) {name=Normal, nodeColor=<rgba 0.1953 1 0 1>}
  NamedRerouteDeclaration(None=TransformPosition(None=Constant3Vector()) {transformSourceType=TRANSFORMPOSSOURCE_Instance, transformType=TRANSFORMPOSSOURCE_World}) {name=InstanceOrigin, nodeColor=<rgba 0 0.8359 1 1>}
  n37 = NamedRerouteUsage()
  n38 = NamedRerouteUsage()
  NamedRerouteDeclaration(None=MaterialFunctionCall(
    VectorToRotate=NamedRerouteUsage()
    LookAtVector=MaterialFunctionCall(Rotation Axis=@n37, Rotation Angle=@n38) {materialFunction=&FixRotateAboutAxisNormals}.New Tangent Space Vertex Normal
   ) {materialFunction=&RotateVector}.Result) {name=AltNormal, nodeColor=<rgba 0.1953 1 0 1>}
  NamedRerouteDeclaration(None=RotateAboutAxis(NormalizedRotationAxis=NamedRerouteUsage(), RotationAngle=NamedRerouteUsage(), PivotPoint=NamedRerouteUsage(), Position=WorldPosition().XYZ)) {name=AltWPO, nodeColor=<rgba 0.5234 0 1 1>}
  n39 = Multiply(A=Time(), B=ScalarParameter() {defaultValue=0.015, group=WPO Sway, parameterName=ConstantRotate_Speed, sortPriority=310})
  n40 = Add(
   A=Add(A=NamedRerouteUsage(), B=@n36)
   B=RotateAboutAxis(
    NormalizedRotationAxis=Constant3Vector() {constant=<rgba 0 0 1 0>}
    RotationAngle=@n39
    PivotPoint=ObjectPositionWS()
    Position=WorldPosition().XYZ
   )
  )
  n41 = Normalize(VectorInput=@n40)
  NamedRerouteDeclaration(None=MaterialFunctionCall(
    VectorToRotate=NamedRerouteUsage()
    LookAtVector=MaterialFunctionCall(Rotation Axis=@n41, Rotation Angle=@n39) {materialFunction=&FixRotateAboutAxisNormals}.New Tangent Space Vertex Normal
   ) {materialFunction=&RotateVector}.Result) {name=SwayRotateNormal, nodeColor=<rgba 0 0.1094 1 1>}
  NamedRerouteDeclaration(None=@n40) {name=SwayRotateWPO, nodeColor=<rgba 0.4297 1 0 1>}
 ]
}
