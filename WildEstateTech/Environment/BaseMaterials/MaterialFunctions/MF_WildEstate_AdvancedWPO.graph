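# Optional material-attributes function for bespoke WildEstate asset movement, adding to World Position Offset and Emissive behind static switches.
# Green vertex color masks an offset that shrinks from Starting_Point to zero with Progress.
# An optional noise-driven axis rotation plus wobble uses two UV masks, which effectively test only U because R is multiplied by itself.
# An optional timer pulses that motion and a Fresnel glow from a custom-primitive-data start time.
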MaterialFunction /WildEstateTech/Environment/BaseMaterials/MaterialFunctions/MF_WildEstate_AdvancedWPO
graph {
 properties {description="Used for bespoke movement that isn't used for most assest.  "}
 nodes [
  n1 = Time()
  n2 = ScalarParameter() {bUseCustomPrimitiveData=true, parameterName=Rotational WPO BP Time, primitiveDataIndex=35}
  n3 = Saturate(None=Divide(A=Subtract(A=@n1, B=@n2), B=ScalarParameter() {defaultValue=2.5, group=WPORotational, parameterName=Rotational WPO Activation Duration}))
  Ceil(None=OneMinus(None=@n3))
  n4 = FunctionInput() {inputType=FunctionInput_MaterialAttributes}
  n5 = GetMaterialAttributes(None=@n4) {attributeGetTypes [#1, #2]}
  n6 = VectorParameter() {group=WPO_Advanced, parameterName=Starting_Point, defaultValue=<rgba 0 0 0 1>}
  n7 = Constant3Vector()
  n8 = Saturate(None=ScalarParameter() {group=WPO_Advanced, parameterName=Progress, sliderMax=1})
  n9 = VertexColor()
  n10 = LinearInterpolate(A=@n5.WorldPositionOffset, B=Add(A=@n5.WorldPositionOffset, B=LinearInterpolate(A=@n6.RGB, B=@n7, Alpha=@n8)), Alpha=Saturate(None=@n9.G))
  n11 = Abs(None=ConstantBiasScale(None=@n3) {bias=-0.5, scale=2})
  n12 = StaticSwitchParameter(
   True=Power(Base=OneMinus(None=@n11), Exp=ScalarParameter() {defaultValue=1, group=WPORotational, parameterName=WPO Soft Timer Power})
   False=Constant() {r=1}
  ) {group=WPORotational, parameterName=Use Rotational WPO Activation Timer, sortPriority=0}
  n13 = MaterialFunctionCall(Float2=TextureCoordinate()) {materialFunction=&BreakOutFloat2Components}
  n14 = AppendVector(
   A=InverseLinearInterpolate(
    A=ScalarParameter() {defaultValue=0.9, group=WPORotational, parameterName=UMaskMin}
    B=ScalarParameter() {defaultValue=1, group=WPORotational, parameterName=UMaskMax}
    Value=@n13.R
   )
   B=InverseLinearInterpolate(
    A=ScalarParameter() {defaultValue=0.57, group=WPORotational, parameterName=VMaskMin}
    B=ScalarParameter() {defaultValue=0.7, group=WPORotational, parameterName=VMaskMax}
    Value=@n13.G
   )
  )
  n15 = Abs(None=ConstantBiasScale(None=Saturate(None=@n14)) {bias=-0.5, scale=2})
  n16 = MaterialFunctionCall(Float2=Ceil(None=OneMinus(None=@n15))) {materialFunction=&BreakOutFloat2Components}
  n17 = Multiply(A=@n16.R, B=@n16.R)
  n18 = ScalarParameter() {group=WPORotational, parameterName=UMaskMin_2}
  n19 = ScalarParameter() {defaultValue=1, group=WPORotational, parameterName=UMaskMax_2}
  n20 = MaterialFunctionCall(Float2=TextureCoordinate()) {materialFunction=&BreakOutFloat2Components}
  n21 = ScalarParameter() {defaultValue=0.95, group=WPORotational, parameterName=VMaskMin_2}
  n22 = ScalarParameter() {defaultValue=1, group=WPORotational, parameterName=VMaskMax_2}
  n23 = Saturate(None=AppendVector(A=InverseLinearInterpolate(A=@n18, B=@n19, Value=@n20.R), B=InverseLinearInterpolate(A=@n21, B=@n22, Value=@n20.G)))
  n24 = Abs(None=ConstantBiasScale(None=@n23) {bias=-0.5, scale=2})
  n25 = MaterialFunctionCall(Float2=Ceil(None=OneMinus(None=@n24))) {materialFunction=&BreakOutFloat2Components}
  n26 = Multiply(A=@n25.R, B=@n25.R)
  n27 = ComponentMask(None=@n23) {r=true}
  n28 = Saturate(None=Power(
    Base=InverseLinearInterpolate(Value=@n27) {constA=1, constB=0}
    Exp=ScalarParameter() {defaultValue=1, group=WPORotational, parameterName=Rotational WPO UV Falloff Power}
   ))
  n29 = Multiply(A=Time(), B=ScalarParameter() {defaultValue=5, group=WPORotational, parameterName=Rotational WPO Speed})
  n30 = Multiply(A=@n29) {constB=0.1}
  n31 = TextureSampleParameter2D(UVs=AppendVector(A=@n30, B=@n29)) {
   group=WPORotational
   parameterName=Rotational WPO Noise
   samplerType=SAMPLERTYPE_Grayscale
   texture=/FortUILibrary/UIKit/Buttons/Textures/LowResBlurredNoise
  }
  n32 = ScalarParameter() {defaultValue=0.45, group=WPORotational, parameterName=Rotational WPO Noise Power}
  n33 = ConstantBiasScale(None=Power(Base=@n31.R, Exp=@n32)) {bias=-0.5, scale=2}
  n34 = Multiply(A=@n33, B=ScalarParameter() {defaultValue=0.02, group=WPORotational, parameterName=Rotational WPO Strength})
  n35 = Transform(None=VectorParameter() {group=RotationalWPO, parameterName=Rotational WPO Axis, defaultValue=<rgba 0 1 0 1>}.RGB) {transformSourceType=TRANSFORMSOURCE_Local}
  n36 = TransformPosition(None=VectorParameter() {group=RotationalWPO, parameterName=Rotational WPO Pivot, defaultValue=<rgba 0 0 120 1>}.RGB) {transformType=TRANSFORMPOSSOURCE_World}
  n37 = Constant()
  n38 = TextureSampleParameter2D(UVs=AppendVector(A=@n29, B=@n30)) {
   group=WPORotational
   parameterName=Rotational WPO Noise
   samplerType=SAMPLERTYPE_Grayscale
   texture=&LowResBlurredNoise
  }
  n39 = Multiply(
   A=ConstantBiasScale(None=Power(Base=@n38.R, Exp=@n32)) {bias=-0.5, scale=2}
   B=ScalarParameter() {defaultValue=0.02, group=WPORotational, parameterName=Rotational WPO Strength X}
  )
  n40 = Multiply(A=@n33, B=ScalarParameter() {defaultValue=0.02, group=WPORotational, parameterName=Rotational WPO Strength Y}) {constB=2}
  n41 = Multiply(
   A=Max(A=@n17, B=Multiply(A=@n26, B=@n28))
   B=Add(
    A=MaterialFunctionCall!(RotationAngle=@n34, NormalizedRotationAxis=@n35, PivotPoint=@n36, Position, Normal, TangentU, TangentV) {materialFunction=/Game/Characters/Player/Male/Medium/Bodies/M_MED_Inferno/FX/Mat/RotateVertexAboutAxis}.WPO
    B=MaterialFunctionCall(R=@n37, G=@n39, B=@n40) {materialFunction=&AppendMany}.RGB
   )
  )
  n42 = StaticSwitchParameter(True=Add(A=@n10, B=Multiply(A=@n12, B=@n41)), False=@n10) {group=WPORotational, parameterName=Use Rotational WPO, sortPriority=-10}
  n43 = ScalarParameter() {defaultValue=2, group=WPORotational, parameterName=WPO Fresnel Exponent}
  n44 = ScalarParameter() {group=WPORotational, parameterName=WPO Fresnel Reflect Fraction}
  n45 = Multiply(
   A=@n17
   B=Multiply(
    A=Fresnel(ExponentIn=@n43, BaseReflectFractionIn=@n44)
    B=VectorParameter() {group=WPORotational, parameterName=WPO Fresnel Color, defaultValue=<rgba 1 1 1 1>}.RGB
   )
  )
  n46 = StaticSwitchParameter(True=Add(A=@n5.EmissiveColor, B=Multiply(A=@n45, B=@n12)), False=@n5.EmissiveColor) {group=WPORotational, parameterName=Use Rotational WPO, sortPriority=-10}
  FunctionOutput(None=StaticSwitchParameter(
    True=SetMaterialAttributes!(MaterialAttributes=@n5.MaterialAttributes, WorldPositionOffset=@n42, EmissiveColor=@n46) {attributeSetTypes [#1, #2]}
    False=@n4
   ) {group=WPO, parameterName=UseAdvancedWPO, sortPriority=0})
  Multiply(A=@n9.G, B=ScalarParameter() {group=WPO_Advanced, parameterName=Enable_WPO_Control})
 ]
}
