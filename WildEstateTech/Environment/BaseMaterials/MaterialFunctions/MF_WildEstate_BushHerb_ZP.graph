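# Material function that, when UseBushHerbZP is on, blends a glowing purple-blue "BushHerb ZP" Wild Estate Tech look into incoming material attributes.
# Emissive layers a Fresnel core with time-pulsed glow under a noise-broken edge color.
# Self-distorted, panning swirl noise on an actor-centered, view-aligned virtual plane drives curve-atlas wave color, omitted at Low and Medium quality.
# The rotated triplanar waves chain ends unconnected to the output.
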MaterialFunction /WildEstateTech/Environment/BaseMaterials/MaterialFunctions/MF_WildEstate_BushHerb_ZP
graph {nodes [
  n1 = PixelNormalWS()
  Add(A=Constant3Vector() {constant=<rgba 0 0 -4 0>}, B=@n1)
  Constant4Vector() {constant=<rgba 0.5887 0.0565 1 1>}
  Constant()
  n2 = NamedRerouteUsage()
  n3 = Time()
  n4 = Constant()
  n5 = ScalarParameter() {defaultValue=0.01, group=Wild Estate Tech Art FX, parameterName=BH_Waves_RotationSpeed, sliderMax=1}
  n6 = ScalarParameter() {defaultValue=-0.01, group=Wild Estate Tech Art FX, parameterName=BH_Waves_RotationSpeed_1, sliderMax=1}
  n7 = Multiply(
   A=LinearInterpolate(
    A=@n3
    B=Multiply(A=@n3, B=@n4)
    Alpha=ScalarParameter() {defaultValue=0.0025, group=Wild Estate Tech Art FX, parameterName=BH_Waves_BackgroundRotation, sliderMax=1}
   )
   B=AppendVector(
    A=AppendVector(A=@n5, B=@n6)
    B=ScalarParameter() {defaultValue=0.022, group=Wild Estate Tech Art FX, parameterName=BH_Waves_RotationSpeed_2, sliderMax=1}
   )
  )
  n8 = TextureObject() {samplerType=SAMPLERTYPE_LinearGrayscale, texture=/Game/Global/Textures/Default/Noise/T_SwirlNoise_N_height}
  n9 = ScalarParameter() {defaultValue=2, group=Wild Estate Tech Art FX, parameterName=BH_Waves_Tiling, sliderMax=2}
  n10 = ScalarParameter() {defaultValue=0.0725, group=Wild Estate Tech Art FX, parameterName=BH_WavesNoise_RotationSpeed, sliderMax=1}
  n11 = ScalarParameter() {defaultValue=-0.0725, group=Wild Estate Tech Art FX, parameterName=BH_WavesNoise_RotationSpeed_1, sliderMax=1}
  n12 = ScalarParameter() {defaultValue=6, group=Wild Estate Tech Art FX, parameterName=BH_Galaxy_AxisFade, sliderMax=8}
  n13 = MaterialFunctionCall!(
   CameraVector=NamedRerouteUsage()
   Rotation Speed=AppendVector(
    A=AppendVector(A=@n10, B=@n11)
    B=ScalarParameter() {defaultValue=0.1, group=Wild Estate Tech Art FX, parameterName=BH_WavesNoise_RotationSpeed_2, sliderMax=1}
   )
   Texture=TextureObject() {samplerType=SAMPLERTYPE_LinearGrayscale, texture=&T_SwirlNoise_N_height}
   Tiling=ScalarParameter() {defaultValue=1, group=Wild Estate Tech Art FX, parameterName=BH_WavesNoise_Tiling, sliderMax=2}
   Offset
   AxisFadeContrast=@n12
  ) {materialFunction=/Game/Weapons/WeaponSkins/MaterialFunction/TriplanarCameraVector_Rotated}
  n14 = Add(
   A=Add(
    A=ScalarParameter() {defaultValue=4, group=Wild Estate Tech Art FX, parameterName=BH_Waves_WarpIntensity, sliderMax=1}
    B=NamedRerouteUsage()
   )
   B=Multiply(
    A=ComponentMask(None=@n13.Result) {r=true}
    B=ScalarParameter() {defaultValue=2.5, group=Wild Estate Tech Art FX, parameterName=BH_WavesNoise_Amount, sliderMax=2}
   ) {constB=0.1}
  )
  n15 = Saturate(None=Desaturation(None=MaterialFunctionCall(CameraVector=@n2, Rotation Speed=@n7, Texture=@n8, Tiling=@n9, Offset=@n14, AxisFadeContrast=@n12) {materialFunction=&TriplanarCameraVector_Rotated}.Result))
  n16 = Constant() {r=5}
  Divide(A=Floor(None=Multiply(A=@n15, B=@n16) {constB=5}), B=Subtract(A=@n16))
  n17 = GetMaterialAttributes(None=NamedRerouteDeclaration(None=FunctionInput() {inputType=FunctionInput_MaterialAttributes}) {name=MatIN, nodeColor=<rgba 1 0 0.1406 1>})
  n18 = ScalarParameter() {defaultValue=0.21, group=Wild Estate Tech Art FX, parameterName=BushHerb_Specular}
  n19 = ScalarParameter() {defaultValue=0.69, group=Wild Estate Tech Art FX, parameterName=BushHerb_Rough}
  n20 = Constant4Vector() {constant=<rgba 0.4477 0.043 0.7604 1>}
  n21 = Constant4Vector() {constant=<rgba 0.1189 0.0183 0.2708 1>}
  n22 = NamedRerouteDeclaration(None=DotProduct(A=CameraVectorWS(), B=@n1)) {name=Fresnel, nodeColor=<rgba 0 1 1 1>}
  n23 = Saturate(None=Power(Base=@n22, Exp=ScalarParameter() {defaultValue=1.7, group=Wild Estate Tech Art FX, parameterName=CoreColor_FresnelExp}))
  n24 = Constant4Vector() {constant=<rgba 0.0448 0.2545 0.7812 1>}
  n25 = ScalarParameter() {defaultValue=1.7, group=Wild Estate Tech Art FX, parameterName=CoreGlow_FresnelExp}
  n26 = ScalarParameter() {defaultValue=3, group=Wild Estate Tech Art FX, parameterName=CoreGlow_FresnelExpMax}
  n27 = Constant() {r=0.5}
  n28 = Saturate(None=LinearInterpolate(
    A=Power(Base=@n22, Exp=@n25)
    B=Power(Base=@n22, Exp=@n26)
    Alpha=MaterialFunctionCall(Speed=@n27) {materialFunction=&DebugTimeSine}.Result
   ) {constA=1, constB=4})
  n29 = Power(
   Base=Multiply(A=NamedRerouteUsage(), B=ScalarParameter() {defaultValue=5, group=Wild Estate Tech Art FX, parameterName=EdgeColor_Falloff}) {constB=5}
   Exp=ScalarParameter() {defaultValue=6, group=Wild Estate Tech Art FX, parameterName=EdgeColor_NoiseContrast}
  ) {constExponent=6}
  n30 = Multiply(A=@n29, B=NamedRerouteUsage())
  n31 = ScalarParameter() {defaultValue=1, group=Wild Estate Tech Art FX, parameterName=EdgeColor_NoiseAmount}
  n32 = LinearInterpolate(
   A=Multiply(A=@n20.RGB, B=ScalarParameter() {defaultValue=1.7, group=Wild Estate Tech Art FX, parameterName=EdgeColor_Intensity}) {constB=1.5}
   B=LinearInterpolate(A=LinearInterpolate(A=@n20.RGB, B=@n21.RGB, Alpha=@n23), B=LinearInterpolate(A=@n21.RGB, B=@n24.RGB) {constAlpha=0.775}, Alpha=@n28)
   Alpha=Saturate(None=LinearInterpolate(A=@n29, B=@n30, Alpha=@n31))
  )
  n33 = NamedRerouteUsage()
  n34 = LinearInterpolate(
   A=Multiply(A=NamedRerouteUsage(), B=ScalarParameter() {defaultValue=0.25, group=Wild Estate Tech Art FX, parameterName=EdgeColor_WavesAmount}) {constB=0.25}
   Alpha=Power(Base=NamedRerouteUsage()) {constExponent=0.2}
  ) {constA=1, constB=0}
  n35 = MaterialFunctionCall(Emissive=QualitySwitch(Default=LinearInterpolate(A=@n32, B=@n33, Alpha=@n34), Low=@n32, Medium=@n32)) {materialFunction=&MF_FX_EmissiveAdjustments}
  n36 = MakeMaterialAttributes!(
   BaseColor=@n21.RGB
   Metallic
   Specular=@n18
   Roughness=@n19
   Anisotropy
   EmissiveColor=@n35.Result
   Opacity
   OpacityMask
   Normal
   Tangent
   WorldPositionOffset
   SubsurfaceColor=Constant()
   ClearCoat
   ClearCoatRoughness
   AmbientOcclusion
   Refraction
   CustomizedUVs_0
   CustomizedUVs_1
   CustomizedUVs_2
   CustomizedUVs_3
   CustomizedUVs_4
   CustomizedUVs_5
   CustomizedUVs_6
   CustomizedUVs_7
   PixelDepthOffset
   ShadingModel
   Displacement
  )
  n37 = NamedRerouteUsage()
  FunctionOutput(None=StaticSwitchParameter(True=BlendMaterialAttributes(A=@n17.MaterialAttributes, B=@n36, Alpha=@n37), False=NamedRerouteUsage()) {group=Wild Estate Tech Art FX, parameterName=UseBushHerbZP, sortPriority=30})
  n38 = Constant3Vector() {constant=<rgba 0 1 0 0>}
  n39 = VectorParameter() {desc="Dont Touch! this Vector is set by the CCPM BP.", group=Wild Estate Tech Art FX, parameterName=BH_ActorRotationAmount, sortPriority=64}
  n40 = Divide(A=Multiply(A=@n39.B, B=ScalarParameter() {defaultValue=0.1, group=Wild Estate Tech Art FX, parameterName=BH_RotationAmount, sliderMax=1}) {constB=0.333}) {constB=360}
  n41 = Constant3Vector()
  n42 = Transform(None=CameraVectorWS()) {transformSourceType=TRANSFORMSOURCE_Camera, transformType=TRANSFORM_Local}
  n43 = MaterialFunctionCall(Float3=Add(A=RotateAboutAxis(NormalizedRotationAxis=@n38, RotationAngle=@n40, PivotPoint=@n41, Position=@n42), B=@n42)) {materialFunction=&BreakOutFloat3Components}
  MaterialFunctionCall(X=@n43.B, Y=@n43.B, Z=@n43.B) {materialFunction=&MakeFloat3}
  Multiply(A=Multiply(A=NamedRerouteUsage()) {constB=1.25})
  Multiply(A=NamedRerouteUsage(), B=@n28)
  Multiply(A=@n9, B=Constant3Vector() {constant=<rgba 3 1.33 0.2 0>})
  NamedRerouteDeclaration(None=FunctionInput() {inputName=Mask, inputType=FunctionInput_Scalar, sortPriority=1}) {name=Mask, nodeColor=<rgba 0.6172 1 0 1>}
  NamedRerouteDeclaration(None=Power(Base=MaterialFunctionCall() {materialFunction=&BoundingBoxBased_0-1_UVW}.B) {constExponent=1.225}) {name=Gradient, nodeColor=<rgba 1 0 0.2578 1>}
  n44 = MaterialFunctionCall(
   Input=MaterialFunctionCall() {materialFunction=/Engine/Functions/Engine_MaterialFunctions02/Utility/FOV}."FOV "
   Input Low=Constant() {r=0.414}
   Input High=Constant() {r=1.091}
   Target Low=Constant()
   Target High=Constant() {r=1}
  ) {materialFunction=&RemapValueRange}
  NamedRerouteDeclaration(None=LinearInterpolate(A=Constant() {r=3}, B=Constant() {r=1}, Alpha=Saturate(None=@n44.Result))) {name=FOVCorrection, nodeColor=<rgba 1 0 0.5391 1>}
  n45 = NamedRerouteUsage()
  n46 = TextureSample(UVs=Panner(Coordinate=Multiply(A=@n45) {constB=1.33}) {speedX=0.018, speedY=0.007}) {samplerType=SAMPLERTYPE_LinearGrayscale, texture=/Game/Global/Textures/Default/Flowmaps/T_StormSwirlMask}
  n47 = Multiply(A=AppendVector(A=@n46.R, B=OneMinus(None=@n46.R))) {constB=0.05}
  n48 = Panner(Coordinate=@n45) {speedX=-0.002, speedY=-0.005}
  n49 = Multiply(
   A=Power(Base=@n46.R) {constExponent=0.85}
   B=TextureSample(UVs=Add(A=@n47, B=@n48)) {samplerType=SAMPLERTYPE_LinearGrayscale, texture=&T_StormSwirlMask}.R
  )
  NamedRerouteDeclaration(None=@n49) {name=Noise, nodeColor=<rgba 1 0 0.6797 1>}
  n50 = Constant3Vector() {constant=<rgba 0 0 -1 0>}
  n51 = Constant3Vector() {constant=<rgba 1 0 0 0>}
  NamedRerouteDeclaration(None=MaterialFunctionCall(
    Plane Normal Axis=Transform(None=@n50) {transformSourceType=TRANSFORMSOURCE_View}
    Plane X Axis=Transform(None=@n51) {transformSourceType=TRANSFORMSOURCE_View}
    Plane Center=ActorPositionWS()
    UV Scale=ScalarParameter() {defaultValue=200, group=Wild Estate Tech Art FX, parameterName=ZPNoise_Scale}
   ) {materialFunction=&VirtualPlaneCoordinates}.UVs) {name=CameraVector, nodeColor=<rgba 1 0.0833 0.3367 1>}
  n52 = CurveAtlasRowParameter(CurveTime=@n49) {
   atlas=/WildEstateTech/Gameplay/Materials/ColorCurves/CA_WildEstate_Tech
   curve=/WildEstateTech/Gameplay/Materials/ColorCurves/CC_WE_BuzzCool_01
   group=Wild Estate Tech Art FX
   parameterName=BH_CC_BuzzCool
  }
  NamedRerouteDeclaration(None=@n52) {name=WavesColor, nodeColor=<rgba 1 0.4688 0 1>}
  n53 = Add(A=@n52.A) {constB=0.65}
  n54 = Saturate(None=OneMinus(None=@n28))
  NamedRerouteDeclaration(None=Saturate(None=Multiply(A=@n53, B=@n54))) {name=WavesA, nodeColor=<rgba 0.1484 0 1 1>}
  SetMaterialAttributes!(MaterialAttributes=@n17.MaterialAttributes, BaseColor=@n21.RGB, EmissiveColor=@n35.Result, Specular=@n18, Roughness=@n19) {attributeSetTypes [
    #1
    #2
    #3
    #4
   ]}
  TransformPosition(None=@n41) {transformType=TRANSFORMPOSSOURCE_World}
  VertexNormalWS()
 ]}
