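# Material function that animates a chest opening through World Position Offset, behind a static switch that otherwise passes material attributes through.
# Vertex color red masks a pivot rotation, translation and height offset, combined with a sine wiggle using height within object bounds and a jump.
# A saturated CODE_Timestamp, or looping debug time, both scaled by ChestSpeed, supplies animation progress.
# Normals also receive a rotation-based adjustment, skipped on forward and mobile shading paths.
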MaterialFunction /WildEstateTech/Environment/BaseMaterials/MaterialFunctions/MF_WildEstate_ChestAnimation
graph {nodes [
  n1 = GetMaterialAttributes(None=FunctionInput() {inputType=FunctionInput_MaterialAttributes}) {attributeGetTypes [#1, #2]}
  n2 = Add(A=@n1.WorldPositionOffset, B=NamedRerouteUsage())
  n3 = Transform(None=@n1.Normal)
  n4 = RotateAboutAxis(NormalizedRotationAxis=NamedRerouteUsage(), RotationAngle=NamedRerouteUsage(), PivotPoint=Constant3Vector(), Position=@n3)
  n5 = LinearInterpolate(
   A=@n1.Normal
   B=Transform(None=Add(A=@n3, B=@n4)) {transformSourceType=TRANSFORMSOURCE_World, transformType=TRANSFORM_Tangent}
   Alpha=NamedRerouteUsage()
  )
  n6 = ShadingPathSwitch(Default=@n5, Deferred=@n5, Forward=@n1.Normal, Mobile=@n1.Normal)
  FunctionOutput(None=StaticSwitchParameter(
    True=SetMaterialAttributes!(MaterialAttributes=@n1.MaterialAttributes, WorldPositionOffset=@n2, Normal=@n6) {attributeSetTypes [#1, #2]}
    False=@n1.MaterialAttributes
   ) {group=Wild Estate Tech Art FX, parameterName=Use WildEstate Chest Animation})
  n7 = NamedRerouteUsage()
  n8 = ScalarParameter() {group=WPO Chest Animation, parameterName=DelayOpen}
  Multiply(A=Divide(A=@n7, B=OneMinus(None=@n8)), B=Subtract(A=@n7, B=@n8))
  n9 = NamedRerouteUsage()
  n10 = Multiply(
   A=VectorParameter() {group=WPO Chest Animation, parameterName=Chest_Wigglestrength, defaultValue=<rgba 3 3 3 1>}.RGB
   B=Sine(None=Multiply(B=@n9) {constA=2})
  )
  n11 = Subtract(
   A=ComponentMask(None=WorldPosition().XYZ) {b=true}
   B=ComponentMask(None=MaterialFunctionCall() {materialFunction=&ObjectPivotPoint}.Object Pivot Location) {b=true}
  )
  n12 = ComponentMask(None=MaterialFunctionCall() {materialFunction=&ObjectLocalBounds}.Local Bounds Size) {b=true}
  n13 = LinearInterpolate(B=Saturate(None=Divide(A=@n11, B=@n12)), Alpha=NamedRerouteUsage())
  n14 = Multiply(A=Time(), B=ScalarParameter() {defaultValue=0.1, group=WPO Chest Animation, parameterName=Chest_Animation_WiggleSpeed})
  n15 = AppendVector(A=Sine(None=@n14), B=Sine(None=Multiply(A=@n14) {constB=1.5}))
  n16 = Sine(None=Multiply(A=@n14) {constB=2})
  n17 = NamedRerouteUsage()
  n18 = ScalarParameter() {group=WPO Chest Animation, parameterName=DelayJump}
  n19 = OneMinus(None=@n18)
  n20 = Saturate(None=Multiply(A=Divide(A=@n17, B=@n19), B=Subtract(A=@n17, B=@n18)))
  n21 = Sine(None=Saturate(None=Divide(A=@n20)))
  n22 = ScalarParameter() {defaultValue=16, group=WPO Chest Animation, parameterName=Chest_Anim_Power2}
  n23 = Add(
   A=Multiply(A=Multiply(A=@n10, B=@n13), B=AppendVector(A=@n15, B=@n16))
   B=Multiply(
    A=VectorParameter() {group=WPO Chest Animation, parameterName=Chest_Anim_JumpHeight, defaultValue=<rgba 0 0 5 1>}.RGB
    B=Power(Base=@n21, Exp=@n22)
   )
  )
  n24 = Transform(None=VectorParameter() {group=WPO Chest Animation, parameterName=Chest_Rotation_Axis, defaultValue=<rgba 1 0 0 1>}.RGB) {transformSourceType=TRANSFORMSOURCE_Local}
  n25 = ScalarParameter() {defaultValue=0.5238, group=WPO Chest Animation, parameterName=Chest_Anim_Power}
  n26 = Multiply(
   A=ScalarParameter() {defaultValue=0.175, group=WPO Chest Animation, parameterName=Chest_Animation_Max_Rotation}
   B=Saturate(None=Power(Base=@n7, Exp=@n25))
  )
  n27 = RotateAboutAxis(
   NormalizedRotationAxis=NamedRerouteDeclaration(None=Normalize(VectorInput=@n24)) {name=ChestRotationAxis, nodeColor=<rgba 1 0 0.8437 1>}
   RotationAngle=NamedRerouteDeclaration(None=@n26) {name=ChestMaxRotation, nodeColor=<rgba 0.5234 1 0 1>}
   PivotPoint=TransformPosition(None=VectorParameter() {group=WPO Chest Animation, parameterName=Chest_Rotation_Pivot, defaultValue=<rgba 0 0 0 1>}.RGB) {transformType=TRANSFORMPOSSOURCE_World}
   Position=WorldPosition().XYZ
  )
  n28 = Multiply(
   A=Normalize(VectorInput=@n26)
   B=Transform(None=VectorParameter() {group=WPO Chest Animation, parameterName=Chest_Move_to_Location, defaultValue=<rgba 0 0 0 1>}.RGB) {transformSourceType=TRANSFORMSOURCE_Local}
  )
  n29 = Multiply(
   A=NamedRerouteDeclaration(None=VertexColor().R) {name=VertexColor, nodeColor=<rgba 1 0 0 1>}
   B=Add(A=NamedRerouteUsage(), B=Add(A=@n27, B=@n28))
  )
  NamedRerouteDeclaration(None=LinearInterpolate(B=Add(A=@n23, B=@n29), Alpha=NamedRerouteUsage())) {name=FinalWPO, nodeColor=<rgba 1 0.5156 0 1>}
  n30 = ScalarParameter() {defaultValue=3, group=WPO Chest Animation, parameterName=ChestSpeed}
  n31 = Multiply(A=Frac(None=Time()), B=@n30) {constB=2}
  n32 = Multiply(A=@n30, B=ScalarParameter() {group=WPO Chest Animation, parameterName=CODE_Timestamp})
  NamedRerouteDeclaration(None=StaticSwitchParameter(True=Saturate(None=@n31), False=Saturate(None=@n32)) {group=WPO Chest Animation, parameterName=DebugChestAnim, sortPriority=0}) {name=DebugChestAnim, nodeColor=<rgba 0.1953 0 1 1>}
  n33 = VectorParameter() {group=WPO Chest Animation, parameterName=WPO_TopMovement_Min, sortPriority=300, defaultValue=<rgba 0 0 0 1>}
  n34 = VectorParameter() {group=WPO Chest Animation, parameterName=WPO_TopMovement_Max, sortPriority=300, defaultValue=<rgba 10 -10 150 1>}
  n35 = NamedRerouteUsage()
  n36 = NamedRerouteDeclaration(None=ConstantBiasScale(None=Cosine(None=@n35) {period=2})) {name=SineTime, nodeColor=<rgba 1 0.6328 0 1>}
  NamedRerouteDeclaration(None=Transform(None=LinearInterpolate(A=@n33.RGB, B=@n34.RGB, Alpha=@n36)) {transformSourceType=TRANSFORMSOURCE_Local}) {name=AddHeight, nodeColor=<rgba 0.9219 1 0 1>}
  NamedRerouteUsage()
  Normalize()
  Sine(None=@n35)
 ]}
