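# Material function that adds wind-like cloth animation to input material attributes behind a static switch, also outputting the normal and additive WPO separately.
# Panning flag textures, on mesh or object-relative projected UVs, drive vertex offset and an angle-corrected normal blend.
# Displacement follows a local direction or the vertex normal, masked by vertex color.
# Phase offsets from actor position, stepped UV1 or left/right side vary the motion.
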MaterialFunction /WildEstateTech/Environment/BaseMaterials/MaterialFunctions/MF_WildEstate_ClothWPO
graph {nodes [
  n1 = NamedRerouteUsage()
  n2 = GetMaterialAttributes(None=NamedRerouteDeclaration(None=FunctionInput() {inputType=FunctionInput_MaterialAttributes, sortPriority=-10}) {name=InMat, nodeColor=<rgba 0.3359 0 1 1>}) {attributeGetTypes [#1, #2]}
  n3 = MaterialFunctionCall(BaseNormal=NamedRerouteDeclaration(None=@n2.Normal) {name=InNormal, nodeColor=<rgba 0 1 0.9297 1>}, AdditionalNormal=NamedRerouteUsage()) {materialFunction=&BlendAngleCorrectedNormals}
  n4 = Add(A=NamedRerouteUsage(), B=NamedRerouteUsage())
  FunctionOutput(None=StaticSwitchParameter(
    True=SetMaterialAttributes!(MaterialAttributes=@n1, Normal=@n3.Result, WorldPositionOffset=@n4) {attributeSetTypes [#1, #2]}
    False=NamedRerouteUsage()
   ) {group=WPO, parameterName=UseWPO_ClothAnimation, sortPriority=-100}) {sortPriority=-10}
  n5 = Constant3Vector() {constant=<rgba 0 0 1 0>}
  n6 = NamedRerouteUsage()
  n7 = ScalarParameter() {defaultValue=1, group=WPO_ClothAnimation, parameterName=WPO_ClothPanDirection}
  n8 = ScalarParameter() {defaultValue=44.8109, group=WPO_ClothAnimation, parameterName=ClothRandomOffsetAmount}
  n9 = WorldPosition() {worldPositionShaderOffset=WPT_ExcludeAllShaderOffsets}
  n10 = ObjectPositionWS()
  n11 = VectorParameter() {group=WPO_ClothAnimation, parameterName=LeftRightDirection, defaultValue=<rgba 1 0 0 1>}
  n12 = Ceil(None=DotProduct(A=Subtract(A=@n9.XYZ, B=@n10), B=Transform(None=@n11.RGB) {transformSourceType=TRANSFORMSOURCE_Local}))
  n13 = Add(A=NamedRerouteUsage(), B=Multiply(A=@n8, B=Saturate(None=@n12)))
  n14 = NamedRerouteUsage()
  n15 = TextureCoordinate() {coordinateIndex=1}
  n16 = Floor(None=Multiply(A=ComponentMask(None=@n15) {r=true}, B=ScalarParameter() {defaultValue=12, group=WPO_ClothAnimation, parameterName=ClothRandomOffsetSteps}))
  n17 = Divide(A=ActorPositionWS(), B=ScalarParameter() {defaultValue=16.71, group=WPO_ClothAnimation, parameterName="ClothAnimOffsetChange(WS)"})
  n18 = ComponentMask(None=@n17) {r=true}
  n19 = ComponentMask(None=@n17) {g=true}
  n20 = StaticSwitchParameter(True=Add(A=@n14, B=Multiply(A=@n16, B=@n8)), False=Add(A=Add(A=@n18, B=@n19), B=@n14)) {group=WPO_ClothAnimation, parameterName=UseClothRandomOffsetUV2, sortPriority=10}
  n21 = Add(
   A=Rotator(
    Coordinate=FunctionInput(Preview=@n6) {bUsePreviewValueAsDefault=true, inputName=UVs, inputType=FunctionInput_Vector2}
    Time=FunctionInput(Preview=@n7) {bUsePreviewValueAsDefault=true, inputName=Pan Direction, inputType=FunctionInput_Scalar}
   )
   B=Multiply(
    A=Constant2Vector() {g=0.33, r=0.1}
    B=StaticSwitchParameter(True=@n13, False=@n20) {group=WPO_ClothAnimation, parameterName=UseClothRandomOffsetLeftRight, sortPriority=10}
   )
  )
  n22 = TextureSample(UVs=@n21) {samplerType=SAMPLERTYPE_Normal, texture=/Game/Environments/Sets/Junkyard/Textures/T_Flag_N}
  n23 = Multiply(
   A=FunctionInput(Preview=ScalarParameter() {defaultValue=1, group=WPO_ClothAnimation, parameterName=WPO_ClothNormalIntensity}) {bUsePreviewValueAsDefault=true, inputName=WPO_ClothNormalIntensity, inputType=FunctionInput_Scalar}
   B=NamedRerouteUsage()
  )
  FunctionOutput(None=NamedRerouteDeclaration(None=LinearInterpolate(A=@n5, B=@n22.RGB, Alpha=@n23)) {name=NormalFinal, nodeColor=<rgba 0.3828 1 0 1>}) {outputName=Normal}
  n24 = TextureSample(UVs=@n21) {texture=/Game/Environments/Sets/Junkyard/Textures/T_Flag_M}
  n25 = NamedRerouteUsage()
  n26 = NamedRerouteUsage()
  n27 = Subtract(A=@n24.G) {constB=0.5}
  n28 = Multiply(A=Multiply(A=Multiply(A=@n24.R, B=@n25), B=NamedRerouteUsage()), B=Add(A=Multiply(A=@n26, B=@n27), B=VertexNormalWS()))
  n29 = Add(
   A=@n24.R
   B=FunctionInput(Preview=ScalarParameter() {group=WPO_ClothAnimation, parameterName=WPO_ClothWPOOffsetInWindDirection}) {bUsePreviewValueAsDefault=true, inputName=WPO_ClothWPOWindTextureOffset, inputType=FunctionInput_Scalar}
  )
  n30 = NamedRerouteUsage()
  n31 = Multiply(A=Multiply(A=Multiply(A=@n29, B=@n30), B=NamedRerouteUsage()), B=NamedRerouteUsage())
  FunctionOutput(None=NamedRerouteDeclaration(None=StaticSwitchParameter(True=@n28, False=@n31) {group=WPO_ClothAnimation, parameterName=UseClothVertexNormal, sortPriority=10}) {name=AdditiveWPOFinal, nodeColor=<rgba 1 0 0.7031 1>}) {outputName=Additive WPO}
  NamedRerouteDeclaration(None=@n2.MaterialAttributes) {name=InMatAtt, nodeColor=<rgba 0.7109 1 0 1>}
  n32 = Transform(None=Subtract(A=WorldPosition() {worldPositionShaderOffset=WPT_ExcludeAllShaderOffsets}.XYZ, B=ObjectPositionWS())) {transformSourceType=TRANSFORMSOURCE_World, transformType=TRANSFORM_Local}
  n33 = AppendVector(
   A=DotProduct(A=@n32, B=VectorParameter() {group=WPO_ClothAnimation, parameterName=WordlProjectionDirection_X, defaultValue=<rgba 0 1 0 1>}.RGB)
   B=DotProduct(A=@n32, B=VectorParameter() {group=WPO_ClothAnimation, parameterName=WordlProjectionDirection_Y, defaultValue=<rgba 0 0 1 1>}.RGB)
  )
  n34 = ScalarParameter() {defaultValue=1, group=WPO_ClothAnimation, parameterName=WPO_ClothTiling_Y}
  n35 = StaticSwitchParameter(True=AppendVector(A=@n34, B=ScalarParameter() {defaultValue=1, group=WPO_ClothAnimation, parameterName=WPO_ClothTiling}), False=@n34) {group=WPO_ClothAnimation, parameterName=UseClothNonUniformTiling, sortPriority=10}
  n36 = StaticSwitchParameter(True=TextureCoordinate() {coordinateIndex=2}, False=TextureCoordinate()) {group=WPO_ClothAnimation, parameterName=UseCloth2ndUVChannel, sortPriority=10}
  NamedRerouteDeclaration(None=StaticSwitchParameter(True=Divide(A=@n33, B=@n35), False=Multiply(A=@n36, B=@n35)) {group=WPO_ClothAnimation, parameterName=UseWorldAlignedClothProjection, sortPriority=0}) {name=UVs, nodeColor=<rgba 0.4297 1 0 1>}
  n37 = ScalarParameter() {defaultValue=1, group=WPO_ClothAnimation, parameterName=WPO_ClothPanSpeed}
  NamedRerouteDeclaration(None=Multiply(A=Time(), B=FunctionInput(Preview=@n37) {bUsePreviewValueAsDefault=true, inputName=Pan Speed, inputType=FunctionInput_Scalar})) {name=Speed, nodeColor=<rgba 1 0 0.75 1>}
  n38 = VertexColor()
  n39 = AppendVector(A=@n38, B=@n38)
  NamedRerouteDeclaration(None=FunctionInput(Preview=ChannelMaskParameter(None=@n39) {group=WPO_ClothAnimation, maskChannel=Green, parameterName=ClothMask, defaultValue=<rgba 0 1 0 0>}.R) {bUsePreviewValueAsDefault=true, inputName="Motion Mask (Defaults to the Red Vertex Color)", inputType=FunctionInput_Scalar}) {name=VertexColor, nodeColor=<rgba 1 0.6094 0 1>}
  n40 = VectorParameter() {group=WPO_ClothAnimation, parameterName=WPO_ClothOffsetDirection, defaultValue=<rgba 1 0 0 1>}
  NamedRerouteDeclaration(None=Transform(None=FunctionInput(Preview=@n40.RGB) {bUsePreviewValueAsDefault=true, inputName=WPO_ClothOffsetDirection}) {transformSourceType=TRANSFORMSOURCE_Local}) {name=WPODirection, nodeColor=<rgba 0.2891 0 1 1>}
  NamedRerouteDeclaration(None=FunctionInput(Preview=ScalarParameter() {defaultValue=10, group=WPO_ClothAnimation, parameterName=WPO_ClothOffsetIntensity}) {bUsePreviewValueAsDefault=true, inputName=WPO_ClothOffsetIntensity, inputType=FunctionInput_Scalar}) {name=Offset, nodeColor=<rgba 0.0312 1 0 1>}
  NamedRerouteDeclaration(None=@n2.WorldPositionOffset) {name=InWPO, nodeColor=<rgba 0 1 0.9766 1>}
 ]}
