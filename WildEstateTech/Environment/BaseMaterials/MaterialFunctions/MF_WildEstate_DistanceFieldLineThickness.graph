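# Material function that turns a distance field texture value into a line mask whose thickness and opacity vary with camera depth.
# Pixel depth, shifted by an offset and divided by a transition distance, drives near-to-far blends of line width and opacity.
# The width acts as a near-hard threshold on the field, and opacity blends that result from white.
# A static IsEnabled switch, off by default, otherwise passes the input through unchanged.
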MaterialFunction /WildEstateTech/Environment/BaseMaterials/MaterialFunctions/MF_WildEstate_DistanceFieldLineThickness
graph {nodes [
  n1 = FunctionInput() {inputName=InDistanceFieldTexture, inputType=FunctionInput_Scalar, sortPriority=1}
  n2 = ComponentMask(None=@n1) {r=true}
  n3 = PixelDepth()
  n4 = FunctionInput(Preview=ScalarParameter() {defaultValue=2000, parameterName=LineFadeOffset}) {bUsePreviewValueAsDefault=true, inputName=LineTransitionOffset, inputType=FunctionInput_Scalar, sortPriority=21}
  n5 = ScalarParameter() {defaultValue=10000, parameterName=LineWidthTransitionDistance}
  n6 = Saturate(None=Divide(
    A=Subtract(A=@n3.R, B=@n4)
    B=FunctionInput(Preview=@n5) {bUsePreviewValueAsDefault=true, inputName=LineTransitionDistance, inputType=FunctionInput_Scalar, sortPriority=12}
   ))
  n7 = LinearInterpolate(
   A=FunctionInput(Preview=ScalarParameter() {defaultValue=0.4045, parameterName=LineWidthNear}) {bUsePreviewValueAsDefault=true, inputName=LineWidthNear, inputType=FunctionInput_Scalar, sortPriority=10}
   B=FunctionInput(Preview=ScalarParameter() {defaultValue=0.5379, parameterName=LineWidthFar}) {bUsePreviewValueAsDefault=true, inputName=LineWidthFar, inputType=FunctionInput_Scalar, sortPriority=11}
   Alpha=@n6
  )
  n8 = Saturate(None=Divide(A=Subtract(A=@n2, B=@n7)) {constB=0.001})
  n9 = ScalarParameter() {defaultValue=0.25, parameterName=LineFadeOpacity}
  n10 = Saturate(None=LinearInterpolate(
    A=ScalarParameter() {defaultValue=0.75, parameterName=LineFadeOpacityNear}
    B=FunctionInput(Preview=@n9) {bUsePreviewValueAsDefault=true, inputName=LineFadeOpacity, inputType=FunctionInput_Scalar, sortPriority=23}
    Alpha=@n6
   ) {constA=1})
  n11 = StaticBool()
  FunctionOutput(None=StaticSwitch(
    True=LinearInterpolate(B=@n8, Alpha=@n10) {constA=1}
    False=@n1
    Value=FunctionInput(Preview=@n11) {bUsePreviewValueAsDefault=true, inputName=IsEnabled, inputType=FunctionInput_StaticBool, sortPriority=5000}
   )) {outputName=Switch_0}
 ]}
