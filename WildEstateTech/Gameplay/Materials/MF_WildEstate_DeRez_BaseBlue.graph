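# Material function supplying the blue base look for the character de-rez death transition, overriding base color, metallic, specular, normal and emissive on an input material.
# A Rim input, defaulting to Fresnel, blends two color parameters into the glow.
# Epic quality reshapes the rim and modulates the glow with a sine-pulsed grid texture.
# A parameter enables panning of that grid, with alternating row bands scrolling in opposite directions.
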MaterialFunction /WildEstateTech/Gameplay/Materials/MF_WildEstate_DeRez_BaseBlue
graph {
 properties {description=Base Colors for the death Transition}
 nodes [
  Constant3Vector() {constant=<rgba 0 3.8199 6 1>}
  Constant3Vector() {constant=<rgba 0.0583 0.3109 0.4 1>}
  Constant()
  Constant()
  n1 = SetMaterialAttributes!(MaterialAttributes)
  n2 = VectorParameter() {group=FlagDissolveFX, parameterName=Character_Derez_Color_01, sortPriority=111, defaultValue=<rgba 0.0583 0.3109 0.4 1>}
  n3 = FunctionInput(Preview=Fresnel() {baseReflectFraction=0, exponent=1}) {bUsePreviewValueAsDefault=true, inputName=Rim, inputType=FunctionInput_Scalar}
  n4 = Saturate(None=@n3)
  n5 = Power(Base=OneMinus(None=@n4)) {constExponent=4}
  n6 = LinearInterpolate(Alpha=Power(Base=@n4)) {constA=0.3, constB=0.9}
  n7 = QualitySwitch(Default=@n3, Low=@n3, High=@n3, Medium=@n3, Epic=Multiply(A=Add(A=@n5, B=@n6)) {constB=0.6})
  n8 = VectorParameter() {group=FlagDissolveFX, parameterName=Character_Derez_Color_02, sortPriority=112, defaultValue=<rgba 0 3.8199 6 1>}
  n9 = Add(A=@n2.RGB, B=Multiply(A=Power(Base=@n7) {constExponent=3}, B=@n8.RGB))
  n10 = TextureCoordinate() {uTiling=2, vTiling=2}
  n11 = Time()
  n12 = Constant2Vector() {r=0.15}
  n13 = ComponentMask(None=TextureCoordinate() {vTiling=32}) {g=true}
  n14 = LinearInterpolate(A=@n12, B=Multiply(A=@n12) {constB=-1}, Alpha=Fmod(A=Ceil(None=@n13), B=Constant() {r=2}))
  n15 = TextureSample(UVs=LinearInterpolate(
    A=TextureCoordinate() {uTiling=2, vTiling=2}
    B=Panner(Coordinate=@n10, Time=@n11, Speed=@n14) {speedX=0.1}
    Alpha=ScalarParameter() {group=FlagDissolveFX, parameterName=UsePannerAnimation, sortPriority=110}
   ) {constAlpha=1}) {
   samplerSource=SSM_Wrap_WorldGroupSettings
   samplerType=SAMPLERTYPE_LinearColor
   texture=/Game/Packages/Fortress_SharedMaterials/Textures/T_Slowfield_GridSquares_01
  }
  n16 = Sine(None=Add(A=@n15.R, B=Time()))
  n17 = Multiply(A=Power(Base=@n9), B=Multiply(A=ConstantBiasScale(None=@n16), B=@n15.R))
  n18 = Multiply(A=@n9) {constB=0.3}
  n19 = SetMaterialAttributes!(
   MaterialAttributes=GetMaterialAttributes(None=FunctionInput(Preview=@n1) {bUsePreviewValueAsDefault=true, inputName=BaseMaterial, inputType=FunctionInput_MaterialAttributes}) {attributeGetTypes [#1, #2]}.MaterialAttributes
   BaseColor=Constant()
   Metallic=Constant()
   Specular=Constant() {r=1}
   EmissiveColor=QualitySwitch(Default=@n9, Low=@n9, High=@n9, Medium=@n9, Epic=Add(A=@n17, B=@n18))
   Normal=Constant() {r=1}
  ) {attributeSetTypes [
    #3
    #4
    #5
    #6
    #1
   ]}
  FunctionOutput(None=@n19)
  MakeMaterialAttributes!(
   BaseColor
   Metallic
   Specular
   Roughness
   Anisotropy
   EmissiveColor
   Opacity
   OpacityMask
   Normal
   Tangent
   WorldPositionOffset
   SubsurfaceColor
   ClearCoat
   ClearCoatRoughness
   AmbientOcclusion
   Refraction
   CustomizedUVs_0
   CustomizedUVs_1
   CustomizedUVs_2
   CustomizedUVs_3
   CustomizedUVs_4
   CustomizedUVs_5
   CustomizedUVs_6
   CustomizedUVs_7
   PixelDepthOffset
   ShadingModel
   Displacement
  )
 ]
}
