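# Material function that applies a flag dissolve effect, blending incoming material attributes toward the Wild Estate DeRez blue material with a Fresnel rim.
# Local position normalized by object bounds is offset by the remapped FlagDissolveFX_Percentage and scaled by softness to drive the dissolve.
# A hard UV V-axis cutoff limits it to the flag region, and a tiled grid texture modulates the blend alpha.
# The UseFlagDissolveFX static switch passes the input through when off.
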MaterialFunction /WildEstateTech/Environment/BaseMaterials/MaterialFunctions/MF_WildEstate_FlagDissolve
graph {nodes [
  Constant() {r=1.6}
  Constant() {r=-1}
  Constant() {r=0.8}
  n1 = GetMaterialAttributes(None=NamedRerouteUsage())
  n2 = MaterialFunctionCall!(BaseMaterial=@n1.MaterialAttributes, Rim=Fresnel(Normal=PixelNormalWS())) {materialFunction=/WildEstateTech/Gameplay/Materials/MF_WildEstate_DeRez_BaseBlue}
  n3 = TextureSample(UVs=TextureCoordinate() {uTiling=2, vTiling=2}) {
   samplerSource=SSM_Wrap_WorldGroupSettings
   samplerType=SAMPLERTYPE_LinearColor
   texture=/Game/Packages/Fortress_SharedMaterials/Textures/T_Slowfield_GridSquares_01
  }
  n4 = ComponentMask(None=@n3.RGB) {g=true, r=true}
  n5 = DotProduct(A=TextureCoordinate(), B=Constant2Vector() {g=1})
  n6 = Multiply(A=OneMinus(None=Add(A=@n5) {constB=0.33, desc=offset for flag mask})) {constB=256}
  n7 = WorldPosition()
  n8 = ObjectPositionWS()
  n9 = Divide(
   A=Subtract(
    A=TransformPosition(None=@n7.XYZ) {transformSourceType=TRANSFORMPOSSOURCE_World}
    B=TransformPosition(None=@n8) {transformSourceType=TRANSFORMPOSSOURCE_World}
   )
   B=ObjectLocalBounds().Half Extents
  )
  n10 = MaterialFunctionCall(
   Input=ScalarParameter() {group=FlagDissolveFX, parameterName=FlagDissolveFX_Percentage, sliderMax=1, sortPriority=102}
   Input Low=Constant()
   Input High=Constant() {r=1}
   Target Low=ScalarParameter() {defaultValue=-1, group=FlagDissolveFX, parameterName=Target_Low}
   Target High=ScalarParameter() {defaultValue=1.6, group=FlagDissolveFX, parameterName=Target_High}
  ) {materialFunction=&RemapValueRange}
  n11 = ComponentMask(None=Multiply(A=Add(A=@n9, B=@n10.Result) {constB=0.5}, B=ScalarParameter() {defaultValue=0.8, group=FlagDissolveFX, parameterName=dissolve_Softness})) {g=true, r=true}
  n12 = Saturate(None=Multiply(A=Saturate(None=@n6), B=Saturate(None=@n11)))
  n13 = Saturate(None=Add(A=Multiply(A=@n4, B=@n12), B=@n12) {constB=0.1})
  FunctionOutput(None=StaticSwitchParameter(True=BlendMaterialAttributes(A=@n1.MaterialAttributes, B=@n2.Result, Alpha=@n13), False=NamedRerouteUsage()) {group=Wild Estate Tech Art FX, parameterName=UseFlagDissolveFX, sortPriority=0})
  NamedRerouteDeclaration(None=FunctionInput() {inputName=Material, inputType=FunctionInput_MaterialAttributes}) {name=Mat, nodeColor=<rgba 1 0.1875 0 1>}
 ]}
