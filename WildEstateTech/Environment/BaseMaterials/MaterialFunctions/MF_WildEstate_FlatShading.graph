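# Material function that applies switchable flat shading to incoming attributes for Wild Estate environment materials.
# Unlit mode darkens base color and drives eye-adaptation-corrected emissive from base color times time-of-day tints, optionally re-adding original emissive.
# Otherwise it replaces normals with a blended sunlight and camera direction, tinting base color by rounded vertex-normal axes.
# MaterialProxyReplace keeps proxies unmodified; the result feeds MF_Test5050Unlit_Gameplay's Disabled input, with original attributes as Enabled.
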MaterialFunction /WildEstateTech/Environment/BaseMaterials/MaterialFunctions/MF_WildEstate_FlatShading
graph {nodes [
  CameraVectorWS()
  CameraVectorWS()
  n1 = FunctionInput() {inputName=Material In, inputType=FunctionInput_MaterialAttributes}
  n2 = GetMaterialAttributes(None=@n1) {attributeGetTypes [#1, #2]}
  n3 = FunctionInput(Preview=Constant() {r=0.5}) {bUsePreviewValueAsDefault=true, inputName=OverrideUnlitShadinBaseColorDarkening, inputType=FunctionInput_Scalar, sortPriority=99}
  n4 = Multiply(A=@n2.BaseColor, B=@n3) {constB=0.5}
  n5 = FunctionInput(Preview=Constant() {r=1}) {bUsePreviewValueAsDefault=true, inputName=OverrideUnlitShadinBaseColorEmissive, inputType=FunctionInput_Scalar, sortPriority=100}
  n6 = Multiply(
   A=MaterialFunctionCall(
    Color_Morning=VectorParameter() {group=FlatShading, parameterName=FlatShadingColor_Morning, defaultValue=<rgba 0.526 0.526 0.526 1>}.RGB
    Color_Day=VectorParameter() {group=FlatShading, parameterName=FlatShadingColor_Day, defaultValue=<rgba 0.7344 0.7344 0.7344 1>}.RGB
    Color_Evening=VectorParameter() {group=FlatShading, parameterName=FlatShadingColor_Evening, defaultValue=<rgba 0.4948 0.4948 0.4948 1>}.RGB
    Color_Night=VectorParameter() {group=FlatShading, parameterName=FlatShadingColor_Night, defaultValue=<rgba 0.1129 0.1433 0.5938 1>}.RGB
   ) {materialFunction=&Color_Time_of_Day}.Result
   B=Multiply(A=@n5, B=@n2.BaseColor)
  )
  n7 = StaticSwitchParameter(True=Add(A=@n6, B=@n2.EmissiveColor), False=@n6) {
   desc="False - Emissive is overriden completely\r\nTrue - The Emissive layers are added back in on top of the TOD flat shading params"
   group=FlatShading
   parameterName=FlatShading_AddEmissive
  }
  n8 = EyeAdaptationInverse(LightValueInput=@n7)
  n9 = GetMaterialAttributes(None=@n1) {attributeGetTypes [#3, #1]}
  n10 = CollectionParameter() {
   collection=&FortniteMaterialParameters
   parameterName=SunlightDirectionalVector
  }
  n11 = Constant3Vector() {constant=<rgba 0 0 -1 0>}
  n12 = Normalize(VectorInput=Add(A=ComponentMask(None=@n10) {b=true, g=true, r=true}, B=Transform(None=@n11) {transformSourceType=TRANSFORMSOURCE_Camera}) {desc=Compromise for Direction light and all othe rlights})
  n13 = StaticSwitch(
   True=@n12
   False=Transform(None=@n12) {transformSourceType=TRANSFORMSOURCE_World, transformType=TRANSFORM_Tangent}
   Value=FunctionInput(Preview=StaticBool()) {bUsePreviewValueAsDefault=true, inputName=IsWorldSpace, inputType=FunctionInput_StaticBool, sortPriority=20}
  )
  n14 = ScalarParameter() {defaultValue=0.9, group=FlatShading, parameterName=TopBottomBrightness}
  n15 = Round(None=ComponentMask(None=VertexNormalWS()) {b=true})
  n16 = ScalarParameter() {defaultValue=0.9, group=FlatShading, parameterName=RightLeftBrightness}
  n17 = VertexNormalWS()
  n18 = Round(None=Abs(None=ComponentMask(None=@n17) {r=true}))
  n19 = Multiply(A=@n9.BaseColor, B=Min(A=LinearInterpolate(B=@n14, Alpha=@n15) {constA=1}, B=LinearInterpolate(B=@n16, Alpha=@n18) {constA=1}))
  n20 = StaticSwitch(
   True=StaticSwitchParameter(
    True=SetMaterialAttributes!(MaterialAttributes=@n2.MaterialAttributes, BaseColor=@n4, EmissiveColor=@n8.EyeAdaptationInverse) {attributeSetTypes [#1, #2]}
    False=SetMaterialAttributes!(MaterialAttributes=@n9.MaterialAttributes, Normal=@n13, BaseColor=@n19) {attributeSetTypes [#3, #1]}
   ) {group=FlatShading, parameterName=UseUnlitFlatShading}
   False=@n1
   Value=FunctionInput(Preview=StaticBool()) {bUsePreviewValueAsDefault=true, inputName=UseFlatShading, inputType=FunctionInput_StaticBool, sortPriority=10}
  )
  FunctionOutput(None=MaterialFunctionCall!(In, Enabled=@n1, Disabled=MaterialProxyReplace(Realtime=@n20, MaterialProxy=@n1), useDirectionalShading, DirectionalShadingMask) {materialFunction=&/MF_Test5050Unlit_Gameplay}.EnabledDisabledOut)
  Transform(None=@n11) {transformSourceType=TRANSFORMSOURCE_Camera, transformType=TRANSFORM_Tangent}
 ]}
