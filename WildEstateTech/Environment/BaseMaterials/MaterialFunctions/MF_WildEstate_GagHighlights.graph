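# Material function for WildEstate environment materials that adds an optional "Gag" highlight to the Emissive of incoming material attributes.
# Emissive blends toward Outline_Color through a named-reroute mask, gated by Gag_Enabled and the UseGagOutline static switch.
# Mask declarations build an animated sheen swept along a direction over world position, multiplied by Fresnel, plus near and far camera-distance fades.
# Quest Fresnel and textured pulse branches end without reaching the output.
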MaterialFunction /WildEstateTech/Environment/BaseMaterials/MaterialFunctions/MF_WildEstate_GagHighlights
graph {nodes [
  n1 = GetMaterialAttributes(None=FunctionInput() {inputType=FunctionInput_MaterialAttributes}) {attributeGetTypes [#1]}
  n2 = VectorParameter() {group=Gag, parameterName=Outline_Color, defaultValue=<rgba 0.8494 0.8862 0.9427 1>}
  n3 = LinearInterpolate(
   A=@n1.EmissiveColor
   B=LinearInterpolate(A=@n1.EmissiveColor, B=@n2.RGB, Alpha=NamedRerouteUsage()) {constAlpha=0.05}
   Alpha=Saturate(None=ScalarParameter() {group=Gag, parameterName=Gag_Enabled, sliderMax=1, sortPriority=30})
  )
  FunctionOutput(None=StaticSwitchParameter(
    True=SetMaterialAttributes!(MaterialAttributes=@n1.MaterialAttributes, EmissiveColor=@n3) {attributeSetTypes [#1]}
    False=@n1.MaterialAttributes
   ) {group=Gag, parameterName=UseGagOutline})
  n4 = Fresnel(
   ExponentIn=LinearInterpolate(
    A=ScalarParameter() {defaultValue=3, group=Gag, parameterName=Quest_Fresnel_EXP_Far}
    B=ScalarParameter() {defaultValue=5, group=Gag, parameterName=Quest_Fresnel_EXP}
    Alpha=NamedRerouteUsage()
   )
   Normal=VertexNormalWS()
  )
  LinearInterpolate(A=Power(Base=@n4) {constExponent=1.2}, B=@n4, Alpha=NamedRerouteUsage())
  n5 = ScalarParameter() {defaultValue=3, group=Gag, parameterName=Pulse_Str}
  LinearInterpolate(A=@n5, B=ScalarParameter() {defaultValue=3, group=Gag, parameterName=Pulse_Str_Far}, Alpha=NamedRerouteUsage())
  n6 = Panner(Coordinate=Divide(A=WorldPosition().Z, B=ScalarParameter() {defaultValue=4000, group=Gag, parameterName=Quest_Noise_Scale})) {speedY=-0.2}
  n7 = Saturate(None=TextureSample(UVs=@n6) {samplerType=SAMPLERTYPE_LinearGrayscale, texture=/Game/Athena/SafeZone/PulseLancerAssets/T_GradientLine_HQ16_SafeZone}.RGB)
  n8 = Saturate(None=TextureSample(UVs=@n6) {samplerType=SAMPLERTYPE_LinearColor, texture=/Game/Athena/SafeZone/Lines}.RGB)
  LinearInterpolate(
   A=Multiply(A=@n2.RGB, B=LinearInterpolate(B=@n5, Alpha=@n7) {constA=1, constB=2})
   B=Multiply(A=@n2.RGB, B=LinearInterpolate(Alpha=@n8))
   Alpha=NamedRerouteUsage()
  )
  LinearInterpolate(Alpha=NamedRerouteUsage()) {constA=2}
  n9 = NamedRerouteDeclaration(None=Fresnel(ExponentIn=ScalarParameter() {defaultValue=2, group=Gag, parameterName=FresnelExponent}) {baseReflectFraction=0.01}) {name=Fresnel_simple, nodeColor=<rgba 0 0.5547 1 1>}
  n10 = Normalize(VectorInput=VectorParameter() {group=Gag, parameterName=SweepDirectionCamSpace, defaultValue=<rgba 0 0 -1 1>}.RGB)
  n11 = DotProduct(A=WorldPosition().XYZ, B=@n10)
  n12 = Panner(
   Coordinate=ComponentMask(None=Multiply(A=@n11, B=@n10)) {b=true, r=true}
   Time=Time()
   Speed=ScalarParameter() {defaultValue=100, group=Gag, parameterName=Speed}
  ) {speedX=0.2, speedY=0.2}
  n13 = OneMinus(None=Divide(A=ComponentMask(None=@n12) {g=true}, B=ScalarParameter() {defaultValue=200, group=Gag, parameterName=Freq}))
  n14 = NamedRerouteDeclaration(None=SmoothStep(Min=ScalarParameter() {defaultValue=0.1, group=Gag, parameterName=Sheen Size}, Max=Constant(), Value=Frac(None=@n13))) {name=M_Sheen, nodeColor=<rgba 0 1 0.4141 1>}
  NamedRerouteDeclaration(None=Multiply(A=Multiply(A=@n9, B=@n14), B=ScalarParameter() {defaultValue=0.3, group=Gag, parameterName=Gag_Line_Str}) {constB=0.3}) {name=M_Sheen_fresnel, nodeColor=<rgba 0 0.8359 1 1>}
  n15 = ActorPositionWS()
  n16 = Subtract(A=Distance(A=@n15, B=CameraPositionWS()), B=ScalarParameter() {defaultValue=4682.3721, group=Gag, parameterName=CameraFadeOffset_Far})
  n17 = ScalarParameter() {defaultValue=2488.0615, group=Gag, parameterName=CameraFadeLength_Far}
  NamedRerouteDeclaration(None=Clamp(None=Divide(A=@n16, B=@n17))) {name=Distance_Far, nodeColor=<rgba 0.9688 1 0 1>}
  n18 = CameraPositionWS()
  n19 = Divide(
   A=Subtract(A=Distance(A=@n15, B=@n18), B=ScalarParameter() {defaultValue=1551.7208, group=Gag, parameterName=CameraFadeOffset_close})
   B=ScalarParameter() {defaultValue=645.5434, group=Gag, parameterName=CameraFadeLength_close}
  )
  NamedRerouteDeclaration(None=OneMinus(None=Clamp(None=@n19))) {name=Distance_close, nodeColor=<rgba 0.7109 0 1 1>}
  NamedRerouteUsage()
  ObjectPositionWS()
 ]}
