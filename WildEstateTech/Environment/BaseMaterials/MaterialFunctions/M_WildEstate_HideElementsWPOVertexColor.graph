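# Wild Estate material function that optionally hides mesh elements through World Position Offset, going by its vertex-color naming.
# It breaks incoming material attributes and passes WPO through the shared Fortnite_HideElementsWithVertexColor function.
# The rebuilt attributes take that WPO and force Ambient Occlusion to 1, while other connected pins all read the BaseColor output as wired.
# Static switch Use_WPO_HideElementsVertexColors selects this result or the untouched input.
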MaterialFunction /WildEstateTech/Environment/BaseMaterials/MaterialFunctions/M_WildEstate_HideElementsWPOVertexColor
graph {nodes [
  n1 = FunctionInput() {inputType=FunctionInput_MaterialAttributes}
  n2 = BreakMaterialAttributes!(Attr=@n1)
  n3 = MaterialFunctionCall!(WPO=@n2.WorldPositionOffset) {materialFunction=/Game/Packages/Fortress_SharedMaterials/Base_Material_Functions/Fortnite_HideElementsWithVertexColor}
  n4 = Constant() {r=1}
  FunctionOutput(None=StaticSwitchParameter(
    True=MakeMaterialAttributes!(
     BaseColor=@n2.BaseColor
     Metallic=@n2.BaseColor
     Specular=@n2.BaseColor
     Roughness=@n2.BaseColor
     Anisotropy=@n2.BaseColor
     EmissiveColor=@n2.BaseColor
     Opacity=@n2.BaseColor
     OpacityMask=@n2.BaseColor
     Normal=@n2.BaseColor
     Tangent=@n2.BaseColor
     WorldPositionOffset=@n3.Result
     SubsurfaceColor=@n2.BaseColor
     ClearCoat=@n2.BaseColor
     ClearCoatRoughness=@n2.BaseColor
     AmbientOcclusion=@n4
     Refraction=@n2.BaseColor
     CustomizedUVs_0=@n2.BaseColor
     CustomizedUVs_1=@n2.BaseColor
     CustomizedUVs_2=@n2.BaseColor
     CustomizedUVs_3=@n2.BaseColor
     CustomizedUVs_4=@n2.BaseColor
     CustomizedUVs_5=@n2.BaseColor
     CustomizedUVs_6=@n2.BaseColor
     CustomizedUVs_7=@n2.BaseColor
     PixelDepthOffset=@n2.BaseColor
     ShadingModel=@n2.BaseColor
     Displacement
    )
    False=@n1
   ) {group=Wild Estate Tech Art FX, parameterName=Use_WPO_HideElementsVertexColors})
 ]}
