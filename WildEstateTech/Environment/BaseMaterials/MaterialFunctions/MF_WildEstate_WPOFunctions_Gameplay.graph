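# Material function for Wild Estate environment materials that bundles bespoke gameplay World Position Offset motion not used by most assets.
# It threads a single Material Attributes input through a serial chain of sub-functions: sway, new wind, fan spin, and wind v2.
# The chain continues through cloth, advanced WPO and chest animation.
# A final vertex-color-based WPO function for hiding elements feeds the output.
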MaterialFunction /WildEstateTech/Environment/BaseMaterials/MaterialFunctions/MF_WildEstate_WPOFunctions_Gameplay
graph {
 properties {description="Used for bespoke movement that isn't used for most assest.  "}
 nodes [
  n1 = FunctionInput() {inputType=FunctionInput_MaterialAttributes}
  n2 = MaterialFunctionCall!(Material Attributes=MaterialFunctionCall!(In=MaterialFunctionCall!(In=@n1) {materialFunction=&/MF_WE_WPOSway}.Result) {materialFunction=&/MF_WE_WPO_WindNew}.Result) {materialFunction=&/MF_WE_FanSpin}
  n3 = MaterialFunctionCall!(In=MaterialFunctionCall!(
    In=MaterialFunctionCall!(In=@n2.Result, Custom Mask, Custom Rotation Axis) {materialFunction=&/MF_WE_WindWPO_v2}.Result
    Pan Speed
    WPO_ClothOffsetDirection
    Pan Direction
    WPO_ClothOffsetIntensity
    "Motion Mask (Defaults to the Red Vertex Color)"
    WPO_ClothWPOWindTextureOffset
    UVs
    WPO_ClothNormalIntensity
   ) {materialFunction=&/MF_WildEstate_ClothWPO}.Result) {materialFunction=&/MF_WildEstate_AdvancedWPO}
  FunctionOutput(None=MaterialFunctionCall!(In=MaterialFunctionCall!(In=@n3.Result) {materialFunction=&/MF_WildEstate_ChestAnimation}.Result) {materialFunction=&/M_WildEstate_HideElementsWPOVertexColor}.Result)
 ]
}
