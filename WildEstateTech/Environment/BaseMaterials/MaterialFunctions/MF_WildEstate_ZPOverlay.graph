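# Wild Estate material function that adds a switchable "Zero Point" overlay effect to input material attributes, also outputting WPO, Emissive and Opacity.
# Fresnel, animated sheen bands and self-distorted storm-swirl noise, projected on a camera-facing plane at actor position, drive the mask and color blends.
# The glow adds to emissive, and the sum is divided by eye adaptation.
# An optional phase-offset sine adds floating world position offset.
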MaterialFunction /WildEstateTech/Environment/BaseMaterials/MaterialFunctions/MF_WildEstate_ZPOverlay
graph {nodes [
  n1 = GetMaterialAttributes(None=NamedRerouteUsage()) {attributeGetTypes [#1, #2]}
  n2 = NamedRerouteUsage()
  n3 = MaterialFunctionCall(
   Brightness=ScalarParameter() {defaultValue=1.2, group=ZP_OverlayFX, parameterName=ZPOverlay.Fresnel.Brightness}
   Exponent=ScalarParameter() {defaultValue=6, group=ZP_OverlayFX, parameterName=ZPOverlay.Fresnel.Exp}
  ) {materialFunction=&MF_Fresnel_FX}
  n4 = Add(A=@n3.Fresnel Inverted, B=ScalarParameter() {defaultValue=0.025, group=ZP_OverlayFX, parameterName=ZPOverlay.MaskFill})
  n5 = NamedRerouteUsage()
  n6 = Multiply(A=@n5, B=ScalarParameter() {defaultValue=0.1, group=ZP_OverlayFX, parameterName=ZPOverlay.TextureBlend})
  n7 = Saturate(None=Add(A=Multiply(A=@n2) {constB=0.05}, B=Add(A=@n4, B=@n6)))
  n8 = Saturate(None=Add(A=@n3.Fresnel Inverted, B=@n5))
  n9 = LinearInterpolate(
   A=VectorParameter() {group=ZP_OverlayFX, parameterName=ZPColorA, defaultValue=<rgba 0.4452 0 0.6584 1>}.RGB
   B=VectorParameter() {group=ZP_OverlayFX, parameterName=ZPColorB, defaultValue=<rgba 0 0.1144 0.9911 1>}.RGB
   Alpha=@n8
  ) {desc=Color}
  n10 = LinearInterpolate(
   A=LinearInterpolate(A=Constant3Vector() {constant=<rgba 0.1263 0.0737 0.2917 1>}, B=Constant3Vector() {constant=<rgba 0.5 0.0465 0.2951 1>}, Alpha=@n8)
   B=@n9
  )
  n11 = Multiply(A=NamedRerouteUsage()) {constB=0.1}
  n12 = LinearInterpolate(
   B=Multiply(A=LinearInterpolate(A=@n9, B=@n10, Alpha=@n11), B=ScalarParameter() {defaultValue=3, group=ZP_OverlayFX, parameterName=ZPOverlay.Brightness})
   Alpha=@n7
  )
  n13 = Divide(A=Add(A=@n1.EmissiveColor, B=@n12), B=EyeAdaptation().EyeAdaptation)
  n14 = Add(A=Time(), B=Constant3Vector() {constant=<rgba 0 0.33 0.7 0>})
  n15 = Constant3Vector() {constant=<rgba 0.09 0.13 0.17 0>}
  n16 = VectorParameter() {group=ZP_OverlayFX, parameterName=ZPOverlay.WPO.Intensity, channelNames {a=Multi, b=Z, g=Y, r=X}, defaultValue=<rgba 2 2 3 1>}
  n17 = Multiply(A=Sine(None=Multiply(A=@n14, B=@n15)), B=Multiply(A=@n16.RGB, B=@n16.RGB))
  n18 = StaticSwitchParameter(True=Add(A=@n1.WorldPositionOffset, B=@n17) {desc=WPO}, False=@n1.WorldPositionOffset) {group=ZP_OverlayFX, parameterName=ZPOverlay.UseFloatWPO, sortPriority=21}
  FunctionOutput(None=StaticSwitchParameter(
    True=SetMaterialAttributes!(MaterialAttributes=@n1.MaterialAttributes, EmissiveColor=@n13, WorldPositionOffset=@n18) {attributeSetTypes [#1, #2]}
    False=NamedRerouteUsage()
   ) {group=Wild Estate Tech Art FX, parameterName=UseZeroPointOverlayFX, sortPriority=20})
  FunctionOutput(None=NamedRerouteUsage()) {outputName=WPO, sortPriority=2}
  FunctionOutput(None=NamedRerouteUsage()) {outputName=Emissive, sortPriority=1}
  FunctionOutput(None=NamedRerouteUsage()) {outputName=Opacity, sortPriority=3}
  NamedRerouteDeclaration(None=FunctionInput() {bUsePreviewValueAsDefault=true, inputName=Material, inputType=FunctionInput_MaterialAttributes}) {name=Mat, nodeColor=<rgba 1 0.1875 0 1>}
  n19 = MaterialFunctionCall(
   Plane Normal Axis=Transform(None=Constant3Vector() {constant=<rgba 0 0 -1 0>}) {transformSourceType=TRANSFORMSOURCE_View}
   Plane X Axis=Transform(None=Constant3Vector() {constant=<rgba 1 0 0 0>}) {transformSourceType=TRANSFORMSOURCE_View}
   Plane Center=ActorPositionWS()
   UV Scale=ScalarParameter() {defaultValue=400, group=ZP_OverlayFX, parameterName=ZP_Scale}
  ) {materialFunction=&VirtualPlaneCoordinates}
  n20 = TextureSample(UVs=Panner(Coordinate=Multiply(A=@n19.UVs) {constB=1.33}) {speedX=0.05, speedY=0.03}) {samplerType=SAMPLERTYPE_LinearGrayscale, texture=/Game/Global/Textures/Default/Flowmaps/T_StormSwirlMask}
  n21 = AppendVector(A=@n20.R, B=OneMinus(None=@n20.R))
  n22 = TextureSample(UVs=Add(A=Multiply(A=@n21) {constB=0.05}, B=Panner(Coordinate=@n19.UVs) {speedX=-0.035, speedY=-0.01})) {samplerType=SAMPLERTYPE_LinearGrayscale, texture=&T_StormSwirlMask}
  NamedRerouteDeclaration(None=Multiply(A=Multiply(A=@n20.R, B=@n22.R), B=ScalarParameter() {defaultValue=1, group=ZP_OverlayFX, parameterName=ZPOverlay.Texture.Multi})) {name=Texture, nodeColor=<rgba 0 0.1328 1 1>}
  NamedRerouteDeclaration(None=@n17) {name=WPO, nodeColor=<rgba 0 1 0.2031 1>}
  NamedRerouteDeclaration(None=@n12) {name=Emissive, nodeColor=<rgba 1 0 0.1172 1>}
  NamedRerouteDeclaration(None=@n7) {name=Opacity, nodeColor=<rgba 1 0 0.3047 1>}
  n23 = Normalize(VectorInput=VectorParameter() {group=ZP_OverlayFX, parameterName=SheenDirectionCamSpace, defaultValue=<rgba 0 0 1 1>}.RGB)
  n24 = DotProduct(A=WorldPosition().XYZ, B=@n23)
  n25 = Panner(
   Coordinate=ComponentMask(None=Multiply(A=@n24, B=@n23)) {b=true, r=true}
   Time=Time()
   Speed=ScalarParameter() {defaultValue=-100, group=ZP_OverlayFX, parameterName=SheenSpeed}
  ) {speedX=0.2, speedY=0.2}
  n26 = OneMinus(None=Divide(A=ComponentMask(None=@n25) {g=true}, B=ScalarParameter() {defaultValue=750, group=ZP_OverlayFX, parameterName=SheenFreq}))
  NamedRerouteDeclaration(None=SmoothStep(Min=ScalarParameter() {defaultValue=0.25, group=ZP_OverlayFX, parameterName=SheenSize}, Max=Constant(), Value=Frac(None=@n26))) {name=Sheen, nodeColor=<rgba 0.0547 1 0 1>}
 ]}
