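# Material function for the Cove Spice equip effect, which reveals a mesh as it grows to full size with a fading pink highlight.
# Time since BP_EquipStartTime fades a view-angle mask that blends pink into base color and emissive.
# The same timer eases a world position offset from a pull toward the object origin back to zero, and rounds into a hard opacity-mask reveal offset by CullTimeOffset.
# A static switch passes the input material through unchanged when disabled.
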MaterialFunction /WildEstateTech/TechArt/Character_Materials/Material_Functions/MF_CoveSpice_EquipMask
graph {nodes [
  n1 = NamedRerouteUsage()
  n2 = Constant3Vector() {constant=<rgba 1 0 0.5802 1>}
  n3 = DotProduct(
   A=VertexNormalWS()
   B=MaterialFunctionCall() {materialFunction=&CameraDirectionVector}.Result
  )
  n4 = Multiply(A=Power(Base=ConstantBiasScale(None=@n3))) {constB=10}
  n5 = Subtract(A=Time(), B=ScalarParameter() {parameterName=BP_EquipStartTime})
  n6 = OneMinus(None=Saturate(None=Multiply(A=@n5) {constB=1.5}))
  n7 = BlendMaterialAttributes(
   A=@n1
   B=SetMaterialAttributes!(MaterialAttributes=@n1, BaseColor=@n2, EmissiveColor=@n2) {attributeSetTypes [#1, #2]}
   Alpha=Multiply(A=Saturate(None=@n4), B=Saturate(None=@n6))
  )
  n8 = TransformPosition(None=Constant3Vector()) {transformType=TRANSFORMPOSSOURCE_World}
  n9 = WorldPosition()
  n10 = Multiply(A=Subtract(A=Time(), B=ScalarParameter() {parameterName=BP_EquipStartTime})) {constB=1.5}
  n11 = LinearInterpolate(
   A=LinearInterpolate(B=Subtract(A=@n8, B=@n9.XYZ), Alpha=ScalarParameter() {defaultValue=0.5, group=Cove Spice, parameterName=BP_EquipMinWPOScale})
   Alpha=Power(Base=Saturate(None=@n10))
  ) {constB=0}
  n12 = Subtract(A=Time(), B=ScalarParameter() {group=Cove Spice, parameterName=BP_EquipStartTime})
  n13 = ScalarParameter() {defaultValue=0.35, desc=When in the anim with this become visible, group=Cove Spice, parameterName=CullTimeOffset}
  n14 = Round(None=Saturate(None=Add(A=@n12, B=@n13)))
  FunctionOutput(None=StaticSwitchParameter(
    True=SetMaterialAttributes!(MaterialAttributes=@n7, WorldPositionOffset=@n11, OpacityMask=@n14) {attributeSetTypes [#3, #4]}
    False=NamedRerouteUsage()
   ) {group=Cove Spice, parameterName=UseCoveSpiceEquipMask})
  NamedRerouteDeclaration(None=FunctionInput(Preview=MakeMaterialAttributes!(
     BaseColor
     Metallic
     Specular
     Roughness
     Anisotropy
     EmissiveColor
     Opacity
     OpacityMask
     Normal
     Tangent
     WorldPositionOffset
     SubsurfaceColor
     ClearCoat
     ClearCoatRoughness
     AmbientOcclusion
     Refraction
     CustomizedUVs_0
     CustomizedUVs_1
     CustomizedUVs_2
     CustomizedUVs_3
     CustomizedUVs_4
     CustomizedUVs_5
     CustomizedUVs_6
     CustomizedUVs_7
     PixelDepthOffset
     ShadingModel
     Displacement
    )) {inputName=Material, inputType=FunctionInput_MaterialAttributes}) {name=MatIn, nodeColor=<rgba 0 1 0.625 1>}
 ]}
