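# Material function that changes monitor screen appearance from GlowDonut state, behind the UseMMOScreenPowerState static switch.
# An SFX channel mask confines changes to BaseColor and Emissive of the incoming material attributes.
# MPC_GlowDonutState drives it: screens darken by a tint while the countdown is hidden, or add flipbook static while PlayMonitorAnims is on.
# MonitorAnimPlayAlpha, rather than time, sets flipbook phase, and a rounded U gradient splits UV offsets between two screens.
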MaterialFunction /WildEstateTech/TechArt/GlowDonut/MF_Screen_MMOPowerState
graph {nodes [
  n1 = GetMaterialAttributes(None=NamedRerouteUsage()) {attributeGetTypes [#1, #2]}
  n2 = VectorParameter() {group=MMO Screen Power State, parameterName=ColorDarkenTint, defaultValue=<rgba 0.6 0.6 0.6 1>}
  n3 = Multiply(A=@n1.BaseColor, B=@n2.RGB)
  n4 = Multiply(
   A=NamedRerouteUsage()
   B=OneMinus(None=CollectionParameter() {collection=/WildEstateTech/TechArt/GlowDonut/Timer/MPC_GlowDonutState, parameterName=CountdownIsVisible})
  )
  n5 = NamedRerouteUsage()
  n6 = CollectionParameter() {collection=&MPC_GlowDonutState, parameterName=PlayMonitorAnims}
  n7 = LinearInterpolate(
   A=@n1.BaseColor
   B=LinearInterpolate(A=LinearInterpolate(A=@n1.BaseColor, B=@n3, Alpha=@n4) {constA=1}, B=Add(A=@n1.BaseColor, B=@n5), Alpha=@n6)
   Alpha=NamedRerouteUsage()
  )
  n8 = Multiply(A=@n1.EmissiveColor, B=@n2.RGB)
  n9 = LinearInterpolate(
   A=@n1.EmissiveColor
   B=LinearInterpolate(A=LinearInterpolate(A=@n1.EmissiveColor, B=@n8, Alpha=@n4) {constA=1}, B=Add(A=@n5, B=@n1.EmissiveColor), Alpha=@n6)
   Alpha=NamedRerouteUsage()
  )
  FunctionOutput(None=StaticSwitchParameter(
    True=SetMaterialAttributes!(MaterialAttributes=@n1.MaterialAttributes, BaseColor=@n7, EmissiveColor=@n9) {attributeSetTypes [#1, #2]}
    False=NamedRerouteUsage()
   ) {group=Wild Estate Tech, parameterName=UseMMOScreenPowerState})
  n10 = Multiply(A=TextureCoordinate(), B=ScalarParameter() {defaultValue=1.5, group=MMO Screen Power State, parameterName=UVMult}) {constB=1.5}
  n11 = VectorParameter() {group=MMO Screen Power State, parameterName=MainUVOffset, defaultValue=<rgba 0 0 0 1>}
  n12 = AppendVector(A=@n11.R, B=@n11.R)
  n13 = Multiply(
   A=Frac(None=Subtract(A=@n10, B=@n12))
   B=AppendVector(
    A=ScalarParameter() {defaultValue=0.7, group=MMO Screen Power State, parameterName=TVStaticScaleX}
    B=ScalarParameter() {defaultValue=0.7, group=MMO Screen Power State, parameterName=TVStaticScaleY}
   )
  )
  n14 = VectorParameter() {
   group=MMO Screen Power State
   parameterName=TV_UVOffset
   sortPriority=12
   channelNames {a=ScreenB.Y, b=ScreenB.X, g=ScreenA.Y, r=ScreenA.X}
   defaultValue=<rgba 0.15 0.15 0 1>
  }
  n15 = Add(
   A=MaterialFunctionCall() {materialFunction=&LinearGradient}.UGradient
   B=ScalarParameter() {group=MMO Screen Power State, parameterName=TVSplitOffset}
  ) {constB=0}
  n16 = LinearInterpolate(
   A=AppendVector(A=@n14.ScreenA.X, B=@n14.ScreenA.X)
   B=AppendVector(A=@n14.ScreenB.X, B=@n14.ScreenB.X)
   Alpha=Saturate(None=Round(None=@n15))
  )
  n17 = VectorParameter() {
   desc=Image sequence params
   group=MMO Screen Power State
   parameterName=TV_ImageSequenceValues
   sortPriority=5
   channelNames {a=Overall Brightness, b=SequenceColumns, g=SequenceRows, r=SequenceSpeed}
   defaultValue=<rgba 1 4 4 1>
  }
  n18 = TextureSampleParameter2D(UVs=MaterialFunctionCall(
    "Animation  Phase (0-1)"=CollectionParameter() {collection=&MPC_GlowDonutState, parameterName=MonitorAnimPlayAlpha}
    Number of Rows=@n17.SequenceRows
    Number of Columns=@n17.SequenceColumns
    UVs=Add(A=@n13, B=@n16)
   ) {materialFunction=&FlipBook}.UVs) {
   group=MMO Screen Power State
   parameterName=OffFlipbookAnim
   sortPriority=0
   texture=/WildEstateTech/TechArt/TVStatic/Textures/T_WildEstate_Static_2x2_04_E
  }
  LinearInterpolate(
   B=VectorParameter() {group=MMO Screen Power State, parameterName=ColorDarkenTint, defaultValue=<rgba 0.6 0.6 0.6 1>}.RGB
   Alpha=Saturate(None=SmoothStep(Value=@n18.B) {constMin=0.2})
  )
  NamedRerouteDeclaration(None=FunctionInput() {inputName=Material, inputType=FunctionInput_MaterialAttributes}) {name=MatIn, nodeColor=<rgba 0 1 0.625 1>}
  n19 = FunctionInput() {inputName=SFX}
  n20 = VectorParameter() {group=MMO Screen Power State, parameterName=SFXScreenMaskChannel, defaultValue=<rgba 0 0 0 1>}
  NamedRerouteDeclaration(None=Saturate(None=DotProduct(A=@n19, B=@n20.RGB))) {name=Mask, nodeColor=<rgba 0 0.1562 1 1>}
  NamedRerouteDeclaration(None=@n18.B) {name=StaticAnim, nodeColor=<rgba 0 1 0.5078 1>}
  NamedRerouteUsage()
  Reroute(None=Multiply(A=Time(), B=@n17.SequenceSpeed))
  ScalarParameter() {group=MMO Screen Power State, parameterName=MMOPowerState_OffAnimAlpha}
 ]}
