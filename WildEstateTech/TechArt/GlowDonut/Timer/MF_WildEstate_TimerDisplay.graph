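# Material function that overlays a GlowDonut countdown timer onto input Material Attributes, also outputting the digit mask.
# MF_WildEstate_CountdownTimer_GlowDonut renders digits from offset UVs and scale parameters.
# The biased mask blends tinted text into BaseColor and Emissive over the dimmed originals, gated by CountdownIsVisible from MPC_GlowDonutState.
# Static switch UseCountdownTimer passes the material through unchanged when off.
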MaterialFunction /WildEstateTech/TechArt/GlowDonut/Timer/MF_WildEstate_TimerDisplay
graph {nodes [
  n1 = FunctionInput() {inputName=Material, inputType=FunctionInput_MaterialAttributes}
  n2 = GetMaterialAttributes(None=@n1) {attributeGetTypes [#1, #2]}
  n3 = VectorParameter() {group=Countdown, parameterName=CountdownTextColor, channelNames {a=Opacity, b="", g="", r=""}, defaultValue=<rgba 1 0.0005 0.0168 1>}
  Add(A=@n2.BaseColor, B=@n3.RGB)
  n4 = ScalarParameter() {group=Countdown, parameterName=CountdownTimer.Offset.X, sortPriority=30}
  n5 = ScalarParameter() {group=Countdown, parameterName=CountdownTimer.Offset.Y, sortPriority=30}
  n6 = MaterialFunctionCall!(
   UVs=Subtract(A=TextureCoordinate(), B=AppendVector(A=@n4, B=@n5))
   UVPadding=AppendVector(
    A=ScalarParameter() {defaultValue=1, group=Countdown, parameterName=CountdownTimer.Scale.X, sortPriority=30}
    B=ScalarParameter() {defaultValue=1, group=Countdown, parameterName=CountdownTimer.Scale.Y, sortPriority=30}
   )
   "NumberofDigits(int)"
  ) {materialFunction=&/MF_WildEstate_CountdownTimer_GlowDonut}
  FunctionOutput(None=@n6.NumbersGreyscale) {outputName=Numbers Mask, sortPriority=1}
  n7 = LinearInterpolate(
   A=Multiply(A=@n2.BaseColor) {constB=0.5}
   B=@n3.RGB
   Alpha=ScalarParameter() {defaultValue=1, group=Countdown, parameterName=CountdownTimer.Color.Blend, sortPriority=30}
  )
  n8 = Multiply(
   A=ScalarParameter() {defaultValue=1, group=Countdown, parameterName=CountdownTimer.Color.Bias, sliderMax=1, sortPriority=30}
   B=@n6.NumbersGreyscale
  )
  n9 = CollectionParameter() {collection=&/MPC_GlowDonutState, parameterName=CountdownIsVisible}
  n10 = LinearInterpolate(A=@n2.BaseColor, B=Saturate(None=LinearInterpolate(A=@n2.BaseColor, B=@n7, Alpha=@n8)), Alpha=@n9)
  n11 = Multiply(A=@n2.EmissiveColor) {constB=0.1}
  n12 = Multiply(A=@n3.RGB, B=ScalarParameter() {defaultValue=1, group=Countdown, parameterName=CountdownTimer.Emissive.Mult, sortPriority=30})
  n13 = LinearInterpolate(A=@n2.EmissiveColor, B=LinearInterpolate(A=@n2.EmissiveColor, B=Add(A=@n11, B=@n12), Alpha=@n8), Alpha=@n9)
  FunctionOutput(None=StaticSwitchParameter(
    True=SetMaterialAttributes!(MaterialAttributes=@n2.MaterialAttributes, BaseColor=@n10, EmissiveColor=@n13) {attributeSetTypes [#1, #2]}
    False=@n1
   ) {group=Wild Estate Tech Art FX, parameterName=UseCountdownTimer}) {outputName=Material}
 ]}
